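# Translucent, unlit, two-sided FX material for the TechnoAwaken body's audio visualizer, usable on Niagara sprites and meshes.
# Radial UVs are scaled and offset, then one axis feeds a procedural edge mask, with no textures sampled directly.
# The refined mask blends two colors, tinted by particle color, for emissive.
# A second refinement, multiplied by a UV clamp mask, inverted Fresnel and particle alpha, forms the translucency fade.
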Material /BRCosmetics/Characters/Player/Male/Medium/Bodies/M_MED_TechnoAwaken_Shake/FX/M_TechnoA_AudioVisualizer_B
graph {
 properties {
  bUsedWithInstancedStaticMeshes=true
  bUsedWithNanite=true
  bUsedWithNiagaraMeshParticles=true
  bUsedWithNiagaraSprites=true
  blendMode=BLEND_Translucent
  pixelDepthOffsetMode=PDOM_Legacy
  shadingModel=MSM_Unlit
  twoSided=true
  usedShadingModels=Unlit
 }
 nodes [
  Constant() {r=1}
  n1 = VectorParameter() {parameterName=Test, defaultValue=<rgba 1 0.9 0 0>}
  n2 = Add(
   A=MaterialFunctionCall(
    UVs=MaterialFunctionCall() {materialFunction=&VectorToRadialValue}.Radial Coordinates
    Texture Scale=AppendVector(A=@n1.R, B=@n1.R)
   ) {materialFunction=&ScaleUVsByCenter}.UVs
   B=AppendVector(A=@n1.B, B=@n1.B)
  )
  n3 = VectorParameter() {group=EdgeFade, parameterName=EdgeFade_EdgeMask_Control, channelNames {a="", b=Power, g=Multiply, r=Offset}, defaultValue=<rgba 0 2 1 1>}
  n4 = MaterialFunctionCall!(
   UorV=MaterialFunctionCall(Float2=@n2) {materialFunction=&BreakOutFloat2Components}.G
   EdgeControl=@n3
  ) {materialFunction=/Game/Effects/Niagara/MasterMaterials/Functions/Common/FXCommon_EdgeMask_Simple}
  n5 = VectorParameter() {parameterName=RFC, defaultValue=<rgba 12 1 0 0.5>}
  n6 = MaterialFunctionCall(Texture=@n4.Mask, RGB=@n5.RGB, A=@n5.RGB) {materialFunction=&MF_FX_Refine}
  n7 = Multiply(
   A=LinearInterpolate(
    A=VectorParameter() {defaultValue=<rgba 0.0167 0 0.1 1>}.RGB
    B=VectorParameter() {parameterName=ParamC, defaultValue=<rgba 0.2 0 0.2 1>}.RGB
    Alpha=Saturate(None=@n6.Result)
   )
   B=Add(A=@n6.Result)
  )
  n8 = ParticleColor()
  MaterialFunctionCall(Emissive=Divide(A=Multiply(A=@n7, B=@n8.RGB)) {constB=1.977}) {materialFunction=&MF_FX_EmissiveAdjustments}
  n9 = VectorParameter() {parameterName=RF, defaultValue=<rgba 1 1 1 0.5>}
  n10 = Multiply(
   A=MaterialFunctionCall!(TexCoord=@n2, ClampChannel=Constant2Vector() {g=1, r=1}) {materialFunction=/Game/Effects/Fort_Effects/MaterialFunctions/MF_FX_TilingMethod}.Result
   B=MaterialFunctionCall(Texture=@n4.Mask, RGB=@n9.RGB, A=@n9.RGB) {materialFunction=&MF_FX_Refine}.Result
  )
  n11 = OneMinus(None=Fresnel(ExponentIn=ScalarParameter() {defaultValue=5, parameterName=FE}) {baseReflectFraction=0})
  n12 = ScalarParameter() {defaultValue=2, parameterName=FP}
  Multiply(A=@n8.A, B=Multiply(A=Multiply(A=@n10), B=Power(Base=@n11, Exp=@n12)))
 ]
}
