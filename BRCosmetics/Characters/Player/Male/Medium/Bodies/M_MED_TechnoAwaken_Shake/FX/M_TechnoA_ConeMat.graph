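# Unlit translucent glow material for Niagara sprite and mesh particles, a cone effect in the TechnoAwaken Shake character cosmetic's FX.
# A soft radial UV mask, shaped by W and scaled by A, hue-shifts a blue color into MF_FX_EmissiveAdjustments.
# A second chain combines that mask, its inverse and an inverted, powered Fresnel into a soft-edged falloff.
# A camera-versus-normal term projected around the instance Z axis ends unused within the graph.
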Material /BRCosmetics/Characters/Player/Male/Medium/Bodies/M_MED_TechnoAwaken_Shake/FX/M_TechnoA_ConeMat
graph {
 properties {
  bUsedWithNiagaraMeshParticles=true
  bUsedWithNiagaraSprites=true
  blendMode=BLEND_Translucent
  pixelDepthOffsetMode=PDOM_Legacy
  shadingModel=MSM_Unlit
  usedShadingModels=Unlit
 }
 nodes [
  n1 = SphereMask(A=TextureCoordinate(), B=Constant() {r=0.5}) {attenuationRadius=0.5, hardnessPercent=0}
  n2 = Power(Base=@n1, Exp=ScalarParameter() {defaultValue=1.5, parameterName=W})
  n3 = Multiply(A=@n2, B=ScalarParameter() {defaultValue=0.1, parameterName=A}) {constB=0.1}
  n4 = VectorParameter() {defaultValue=<rgba 0 0.412 1.2 1>}
  MaterialFunctionCall(Emissive=Divide(A=MaterialFunctionCall(Hue Shift Percentage=@n3, Texture=@n4.RGB) {materialFunction=&HueShift}.Result) {constB=1.977}) {materialFunction=&MF_FX_EmissiveAdjustments}
  n5 = Multiply(A=OneMinus(None=@n1)) {constB=2}
  n6 = Power(Base=OneMinus(None=Fresnel() {baseReflectFraction=0, exponent=1})) {constExponent=6}
  Multiply(A=Multiply(A=Saturate(None=@n5), B=Multiply(A=@n2, B=@n6))) {constB=2}
  n7 = Transform(None=Constant3Vector() {constant=<rgba 0 0 1 0>}) {transformSourceType=TRANSFORMSOURCE_Instance}
  n8 = CameraVectorWS()
  n9 = DotProduct(
   A=PixelNormalWS()
   B=Normalize(VectorInput=MaterialFunctionCall!(Plane Normal=@n7, Vector=@n8) {materialFunction=/Engine/Functions/Engine_MaterialFunctions02/Math/ProjectVectorOntoPlane}.Result)
  )
  OneMinus(None=@n9)
  Power(Base=Abs(None=@n9))
 ]
}
