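# Unlit, translucent, two-sided Niagara ribbon and mesh-particle material drawing animated line streaks for the TechnoAwaken_Shake character cosmetic's FX.
# Particle color is hue-shifted by panning Perlin noise, with strength set per particle, then passed through emissive adjustment.
# Opacity math combines a noise-distorted, sharpened line texture with UV edge masks, inverse Fresnel and particle alpha.
# Screen-space sphere masks around two dynamic-parameter points cut out the line, using X in place of Y.
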Material /BRCosmetics/Characters/Player/Male/Medium/Bodies/M_MED_TechnoAwaken_Shake/FX/M_TechnoA_Line
graph {
 properties {
  bUsedWithNiagaraMeshParticles=true
  bUsedWithNiagaraRibbons=true
  blendMode=BLEND_Translucent
  pixelDepthOffsetMode=PDOM_Legacy
  shadingModel=MSM_Unlit
  twoSided=true
  usedShadingModels=Unlit
 }
 nodes [
  Constant()
  MaterialFunctionCall() {materialFunction=&LocalPosition}
  n1 = TextureObjectParameter() {
   group=Hue
   parameterName=Hue_Texture
   samplerType=SAMPLERTYPE_LinearGrayscale
   sortPriority=0
   texture=&T_SimplePerlin_Greyscale
  }
  n2 = VectorParameter() {group=Hue, parameterName=Hue_UVControl, channelNames {a=VPan, b=UPan, g=VScale, r=UScale}, defaultValue=<rgba 0.3 0.3 0.7 0.7>}
  n3 = MaterialFunctionCall(Texture=@n1, UVControl_RGB=@n2, UVControl_A=@n2, "[Override]UVInput"=ComponentMask(None=TextureCoordinate()) {g=true, r=true}) {materialFunction=&FXCommon_TextureSample}
  n4 = ScalarParameter() {defaultValue=-0.3, group=Hue, parameterName=Hue_GrayAdjust_Add}
  n5 = DynamicParameter() {defaultValue=<rgba 0 1 0.3 1>, paramNames [NoisePan,Param2,HueShiftStr,Param4]}
  n6 = ParticleColor()
  n7 = MaterialFunctionCall(Hue Shift Percentage=Multiply(A=Add(A=@n3.R, B=@n4), B=@n5.HueShiftStr), Texture=@n6.RGB) {materialFunction=&HueShift}
  MaterialFunctionCall(Emissive=Divide(A=@n7.Result) {constB=1.977}) {materialFunction=&MF_FX_EmissiveAdjustments}
  n8 = TextureObjectParameter() {
   group=LineTexture
   parameterName=LineTexture_Texture
   samplerType=SAMPLERTYPE_LinearColor
   sortPriority=0
   texture=/Game/Characters/Player/Male/Medium/Bodies/M_MED_TechnoAwaken_Shake/FX/T_LineStrip
  }
  n9 = TextureObjectParameter() {
   group=Noise
   parameterName=Noise_Texture
   samplerType=SAMPLERTYPE_Normal
   sortPriority=0
   texture=/Game/Characters/Player/Female/Medium/Bodies/F_MED_Bistro_Astronaut/FX/T_GaussianNoise_Normal
  }
  n10 = AppendVector(A=@n5.Param2, B=@n5.Param2)
  n11 = VectorParameter() {group=Noise, parameterName=Noise_UVControl, channelNames {a=VPan, b=UPan, g=VScale, r=UScale}, defaultValue=<rgba 1 0.1 0 0>}
  n12 = VectorParameter() {
   group=Noise
   parameterName=Noise_NoiseControl
   sortPriority=60
   channelNames {a=VStrength, b=UStrength, g=VOffset, r=UOffset}
   defaultValue=<rgba 0 0 0 0.3>
  }
  n13 = VectorParameter() {group=LineTexture, parameterName=LineTexture_UVControl, channelNames {a=VPan, b=UPan, g=VScale, r=UScale}, defaultValue=<rgba 15 2 0 -0.24>}
  n14 = MaterialFunctionCall(
   Texture=@n8
   AdditionalUVOffset=MaterialFunctionCall!(
    NoiseValue=MaterialFunctionCall(Texture=@n9, AdditionalUVOffset=@n10, UVControl_RGB=@n11, UVControl_A=@n11) {materialFunction=&FXCommon_TextureSample}.RG
    Control_RGB=@n12
    Control_A=@n12
   ) {materialFunction=/Game/Effects/Niagara/MasterMaterials/Functions/Common/FXCommon_NoiseControl}.NoiseValue
   UVControl_RGB=@n13
   UVControl_A=@n13
  ) {materialFunction=&FXCommon_TextureSample}
  n15 = MaterialFunctionCall(Float2=TextureCoordinate()) {materialFunction=&BreakOutFloat2Components}
  n16 = VectorParameter() {group=Mask, parameterName=Mask_EdgeMask_ControlA, channelNames {a="", b=Power, g=Multiply, r=Offset}, defaultValue=<rgba 0 1 2 1>}
  n17 = VectorParameter() {group=Mask, parameterName=Mask_EdgeMask_ControlB, channelNames {a="", b=Power, g=Multiply, r=Offset}, defaultValue=<rgba 0 1 2 1>}
  n18 = Multiply(
   A=Power(Base=Multiply(A=@n14.R) {constB=1.5}) {constExponent=8}
   B=Multiply(
    A=MaterialFunctionCall!(UorV=@n15.R, EdgeControl=@n16) {materialFunction=/Game/Effects/Niagara/MasterMaterials/Functions/Common/FXCommon_EdgeMask_Simple}.Mask
    B=MaterialFunctionCall(UorV=@n15.G, EdgeControl=@n17) {materialFunction=&FXCommon_EdgeMask_Simple}.Mask
   )
  )
  n19 = Power(Base=OneMinus(None=Fresnel() {baseReflectFraction=0, exponent=3.5}))
  n20 = ScreenPosition()
  n21 = DynamicParameter() {parameterIndex=1, paramNames [P1X,P1Y,P1Z,Param4]}
  n22 = AppendVector(A=AppendVector(A=@n21.P1X, B=@n21.P1X), B=@n21.P1Z)
  n23 = VectorParameter() {parameterName=wfw, defaultValue=<rgba 6 0 0 1>}
  n24 = OneMinus(None=SphereMask(
    A=@n20.ViewportUV
    B=MaterialFunctionCall!(Center Point World Position=@n22) {materialFunction=/Engine/Functions/Engine_MaterialFunctions02/UVs/WorldSpaceAlignedScreenCoordinates}.Offset UVs
    Radius=@n23.R
    Hardness=@n23.R
   ))
  n25 = DynamicParameter() {parameterIndex=2, paramNames [P2X,P2Y,P2Z,Param4]}
  n26 = AppendVector(A=AppendVector(A=@n25.P2X, B=@n25.P2X), B=@n25.P2Z)
  n27 = OneMinus(None=SphereMask(
    A=@n20.ViewportUV
    B=MaterialFunctionCall(Center Point World Position=@n26) {materialFunction=&WorldSpaceAlignedScreenCoordinates}.Offset UVs
    Radius=@n23.R
    Hardness=@n23.R
   ))
  n28 = Multiply(A=Multiply(A=@n6.A, B=Multiply(A=@n18, B=@n19)), B=Power(Base=Multiply(A=@n24, B=@n27), Exp=ScalarParameter() {defaultValue=4}))
  n29 = VectorParameter() {parameterName=ShouldLineFade, defaultValue=<rgba -70 30 0.1 1>}
  n30 = WorldPosition()
  n31 = CameraPositionWS()
  n32 = Divide(A=Add(A=Distance(A=@n30.XYZ, B=@n31), B=@n29.R), B=@n29.G)
  Multiply(A=@n28, B=LinearInterpolate(A=@n29.B, B=@n29.B, Alpha=Saturate(None=@n32)))
  Multiply(A=@n7.Result, B=@n6.RGB)
  Multiply(A=@n28, B=@n27)
  ScalarParameter() {defaultValue=0.3, group=Hue, parameterName=Hue_GrayAdjust_Mul}
  VectorParameter() {parameterName=C, defaultValue=<rgba 0 1 2 1>}
 ]
}
