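# Unlit additive material for small-block FX on the TechnoAwaken character cosmetic, usable on Niagara mesh particles and ribbons.
# An HDR cyan-blue color parameter blends from a Fresnel rim to a dim flat tint as the instance's Z axis aligns with the camera direction, in either sign.
# Particle color tints the scaled result, which passes through MF_FX_EmissiveAdjustments; particle alpha is separately halved.
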Material /BRCosmetics/Characters/Player/Male/Medium/Bodies/M_MED_TechnoAwaken_Shake/FX/M_TechnoA_SmallBlock
graph {
 properties {
  bUsedWithNiagaraMeshParticles=true
  bUsedWithNiagaraRibbons=true
  blendMode=BLEND_Additive
  pixelDepthOffsetMode=PDOM_Legacy
  shadingModel=MSM_Unlit
  usedShadingModels=Unlit
 }
 nodes [
  n1 = VectorParameter() {defaultValue=<rgba 0 4.5 10 1>}
  n2 = Constant3Vector() {constant=<rgba 0 0 1 0>}
  n3 = DotProduct(
   A=Normalize(VectorInput=Transform(None=@n2) {transformSourceType=TRANSFORMSOURCE_Instance})
   B=MaterialFunctionCall() {materialFunction=&CameraDirectionVector}.Result
  )
  n4 = LinearInterpolate(A=Multiply(A=Fresnel() {baseReflectFraction=0, exponent=3}, B=@n1.RGB), B=Multiply(A=@n1.RGB) {constB=0.1}, Alpha=Power(Base=Abs(None=@n3))) {constB=0.2}
  n5 = ParticleColor()
  MaterialFunctionCall(Emissive=Divide(A=Multiply(A=@n4, B=@n5.RGB)) {constB=1.977}) {materialFunction=&MF_FX_EmissiveAdjustments}
  Multiply(A=Constant() {r=0.5}, B=@n5.A)
 ]
}
