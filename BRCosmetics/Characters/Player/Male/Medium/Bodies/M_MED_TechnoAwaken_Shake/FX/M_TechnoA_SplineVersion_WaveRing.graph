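# Translucent unlit two-sided Niagara mesh-particle material for the wave ring effect on the TechnoAwaken Shake player cosmetic.
# Panning noise displaces vertices along normals, scaled by a dynamic parameter, while SplineThicken sets line width from another.
# Particle color is hue-shifted by instance-space noise for emissive, and particle alpha multiplies a refined UV edge mask.
# A sine-pulsed sound-wave spike branch has no visible consumer in the graph.
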Material /BRCosmetics/Characters/Player/Male/Medium/Bodies/M_MED_TechnoAwaken_Shake/FX/M_TechnoA_SplineVersion_WaveRing
graph {
 properties {
  bUsedWithNiagaraMeshParticles=true
  blendMode=BLEND_Translucent
  pixelDepthOffsetMode=PDOM_Legacy
  shadingModel=MSM_Unlit
  twoSided=true
  usedShadingModels=Unlit
 }
 nodes [
  n1 = DynamicParameter() {paramNames [Param1,W,Param3,Param4]}
  n2 = Multiply(A=VertexNormalWS(), B=@n1.Param1)
  n3 = TextureObjectParameter() {
   group=Noise
   parameterName=Noise_Texture
   samplerType=SAMPLERTYPE_LinearGrayscale
   sortPriority=0
   texture=&T_SimplePerlin_Greyscale
  }
  n4 = VectorParameter() {group=Noise, parameterName=Noise_UVControl, channelNames {a=VPan, b=UPan, g=VScale, r=UScale}, defaultValue=<rgba 0.01 0.5 20 0>}
  n5 = MaterialFunctionCall(Texture=@n3, UVControl_RGB=@n4, UVControl_A=@n4) {materialFunction=&FXCommon_TextureSample}
  n6 = VectorParameter() {
   group=Noise
   parameterName=Noise_NoiseControl
   sortPriority=60
   channelNames {a=VStrength, b=UStrength, g=VOffset, r=UOffset}
   defaultValue=<rgba -0.5 -0.5 50 0>
  }
  n7 = Multiply(
   A=@n2
   B=MaterialFunctionCall(Float2=MaterialFunctionCall!(NoiseValue=@n5.R, Control_RGB=@n6, Control_A=@n6) {materialFunction=/Game/Effects/Niagara/MasterMaterials/Functions/Common/FXCommon_NoiseControl}.NoiseValue) {materialFunction=&BreakOutFloat2Components}.R
  )
  n8 = ScalarParameter() {defaultValue=2, parameterName=SW}
  n9 = Multiply(A=@n8, B=@n1.W)
  n10 = Multiply(A=@n2)
  n11 = Add(A=Add(A=@n7, B=WorldPosition() {worldPositionShaderOffset=WPT_ExcludeAllShaderOffsets}.XYZ), B=@n10)
  Add(
   A=Add(
    A=@n7
    B=MaterialFunctionCall(WidthBase=@n9, WidthTip=@n9, WorldPosition=@n11) {materialFunction=&SplineThicken}.WorldPositionOffset
   )
   B=@n10
  )
  Constant3Vector() {constant=<rgba 0.7772 0 1 1>}
  n12 = TextureObjectParameter() {
   group=Hue
   parameterName=Hue_Texture
   samplerType=SAMPLERTYPE_LinearGrayscale
   sortPriority=0
   texture=&T_SimplePerlin_Greyscale
  }
  n13 = VectorParameter() {group=Hue, parameterName=Hue_UVControl, channelNames {a=VPan, b=UPan, g=VScale, r=UScale}, defaultValue=<rgba 0.0005 0.0005 0 300>}
  n14 = ComponentMask(None=TransformPosition(None=WorldPosition().XYZ) {transformSourceType=TRANSFORMPOSSOURCE_World, transformType=TRANSFORMPOSSOURCE_Instance}) {g=true, r=true}
  n15 = Multiply(
   A=Add(
    A=MaterialFunctionCall(Texture=@n12, UVControl_RGB=@n13, UVControl_A=@n13, "[Override]UVInput"=@n14) {materialFunction=&FXCommon_TextureSample}.R
    B=ScalarParameter() {defaultValue=-0.5, group=Hue, parameterName=Hue_GrayAdjust_Add}
   )
   B=ScalarParameter() {defaultValue=0.5, group=Hue, parameterName=Hue_GrayAdjust_Mul}
  )
  n16 = ParticleColor()
  MaterialFunctionCall(Emissive=Divide(A=MaterialFunctionCall(Hue Shift Percentage=@n15, Texture=@n16.RGB) {materialFunction=&HueShift}.Result) {constB=1.977}) {materialFunction=&MF_FX_EmissiveAdjustments}
  Multiply(A=@n8, B=@n1.Param1)
  n17 = VectorParameter() {group=E, parameterName=E_EdgeMask_Control, channelNames {a="", b=Power, g=Multiply, r=Offset}, defaultValue=<rgba 0 2 2 1>}
  n18 = MaterialFunctionCall!(
   UorV=MaterialFunctionCall(Float2=TextureCoordinate()) {materialFunction=&BreakOutFloat2Components}.R
   EdgeControl=@n17
  ) {materialFunction=/Game/Effects/Niagara/MasterMaterials/Functions/Common/FXCommon_EdgeMask_Simple}
  n19 = VectorParameter() {parameterName=RF, defaultValue=<rgba 25 10 1 0.5>}
  Multiply(
   A=@n16.A
   B=Saturate(None=MaterialFunctionCall(Texture=@n18.Mask, RGB=@n19.RGB, A=@n19.RGB) {materialFunction=&MF_FX_Refine}.Result)
  )
  n20 = TextureObjectParameter() {
   group=Spike
   parameterName=Spike_Texture
   samplerType=SAMPLERTYPE_LinearGrayscale
   sortPriority=0
   texture=&/T_SoundWave_GS
  }
  n21 = VectorParameter() {group=Spike, parameterName=Spike_UVControl, channelNames {a=VPan, b=UPan, g=VScale, r=UScale}, defaultValue=<rgba 2 1 0.2 0>}
  n22 = MaterialFunctionCall(
   Texture=@n20
   UVControl_RGB=@n21
   UVControl_A=@n21
   "[Override]UVInput"=MaterialFunctionCall(XY=TextureCoordinate()) {materialFunction=&Swizzle}.YX
  ) {materialFunction=&FXCommon_TextureSample}
  n23 = VectorParameter() {
   group=Spike
   parameterName=Spike_NoiseControl
   sortPriority=60
   channelNames {a=VStrength, b=UStrength, g=VOffset, r=UOffset}
   defaultValue=<rgba -0.3 -0.5 50 0>
  }
  Multiply(
   A=MaterialFunctionCall(Float2=MaterialFunctionCall(NoiseValue=@n22.RG, Control_RGB=@n23, Control_A=@n23) {materialFunction=&FXCommon_NoiseControl}.NoiseValue) {materialFunction=&BreakOutFloat2Components}.R
   B=Sine(None=Time()) {period=0.5}
  ) {constB=100}
 ]
}
