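# Niagara system of reactive glow, ring, line and wave effects attached to skeletal mesh sockets on the TechnoAwaken Shake character body cosmetic.
# System.ReactiveValue drives glow brightness and PostLine/PostWave spawn rates, which vary by platform.
# System.DoOnce kills those particles and fires burst rings, with a delayed echo, plus flash variants.
# Post emitters pass world positions to materials, while time-of-day, front-end preview and event-audio amplitude adjust several emitters.
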NiagaraSystem /BRCosmetics/Characters/Player/Male/Medium/Bodies/M_MED_TechnoAwaken_Shake/FX/NS_TechnoA_2820TOD
dependencies {
 usedDataInterfaces [
  &NiagaraDataInterfaceArrayFloat2
  &NiagaraDataInterfaceArrayInt32
  &NiagaraDataInterfaceCurve
  &NiagaraDataInterfaceArrayFloat
  &NiagaraDataInterfaceArrayFloat3
  &NiagaraDataInterfaceColorCurve
  &NiagaraDataInterfacePlatformSet
  &NiagaraDataInterfaceArrayColor
 ]
 usedDynamicInputs [
  &InvertBool
  &ClampFloat
  &Divide_Float
  &ModuloFloat
 ]
 usedModules [
  &self:SystemSpawnScript.SystemScriptSource.SystemScriptGraph.NiagaraNodeAssignment_0.SetVariables_#1
  &SystemState
  &self:SystemSpawnScript.SystemScriptSource.SystemScriptGraph.NiagaraNodeAssignment_1.SetVariables_#2
  &self:SystemSpawnScript.SystemScriptSource.SystemScriptGraph.NiagaraNodeAssignment_2.SetVariables_#3
  &DoOnce
 ]
}
emitters {
 BurstRing {
  emitterSpawnScript [/SetVariables(Emitter.float=-100) {isSetParameters=true}]
  emitterUpdateScript [
   &EmitterState(
    Inactive Response=::NewEnumerator1
    Life Cycle Mode=::NewEnumerator1
    Loop Behavior=::NewEnumerator1
    Loop Duration Mode=::NewEnumerator1
    UseLoopDelay=false
   )
   &SpawnRate(
    SpawnRate=&MakeCustomFloatFromBool(Bool=&SetBoolByFloatComparison(A=$NPC.FortniteNPC.EventAudio_Amplitude, B=0.3), True Float=2, False Float=0)
    Use Spawn Probability=false
   ) {enabled=false}
   &SpawnPerFrame(Spawn:NiagaraBool=$System.DoOnce.Execute, Spawn Group=1, Use Spawn Probability=false)
   &SpawnPerFrame(Spawn:NiagaraBool=$System.DoOnce.Execute, Spawn Group=4, Use Spawn Probability=false) {module=SpawnPerFrame001}
   /SetVariables(Emitter.float:NiagaraFloat=&MakeCustomFloatFromBool(
     Bool=$System.DoOnce.Execute
     True Float=&self:ScratchDynamicInput(float:NiagaraFloat=$Emitter.Age, float001:NiagaraFloat=0.16)
     False Float=$Emitter.float
    )) {isSetParameters=true}
   &SpawnBurst_Instantaneous(Spawn Group=1, Spawn Time=$Emitter.float, Use Spawn Probability=false, UseLoopCountLimit=false)
   &SpawnBurst_Instantaneous(Spawn Group=4, Spawn Time=$Emitter.float, Use Spawn Probability=false, UseLoopCountLimit=false) {module=SpawnBurst_Instantaneous001}
  ]
  particleSpawnScript [
   &InitializeParticle(
    Color=<rgba 1.6667 0 20 0.4>
    Lifetime=&Multiply_Float(
     A=&MakeCustomFloatFromBool(
      Bool=&SetBoolByFloatComparison(A=&Subtract_Float(A=$Emitter.Age, B=$Emitter.float), B=-0.01)
      True Float=1.2
      False Float=0.7
     )
     B=&FN_FrontEndFloatModifier(Gameplay=1, Front End=0.7)
    )
    Mesh Scale Mode=::NewEnumerator3
    Mesh Uniform Scale=&MakeCustomFloatFromBool(
     Bool=&SetBoolByFloatComparison(A=&Subtract_Float(A=$Emitter.Age, B=$Emitter.float), B=-0.01)
     True Float=0.08
     False Float=0.08
    )
    Sprite Size Mode=::NewEnumerator3
    UsePositionOffset=false
    Write Material Random=true
   )
   &Color(Color=&MultiplyLinearColor(
     Linear Color=$Particles.Color
     Scale Factor=&MakeVector4(
      X=1
      Y=1
      Z=1
      W=&MakeCustomFloatFromBool(
       Bool=&SetBoolByFloatComparison(A=&Subtract_Float(A=$Emitter.Age, B=$Emitter.float), B=-0.01)
       True Float=0.2
       False Float=1
      )
     )
    ))
   /SetVariables(Particles.bool:NiagaraBool=&SetBoolByFloatComparison(A=&Subtract_Float(A=$Emitter.Age, B=$Emitter.float), B=-0.01)) {isSetParameters=true}
  ]
  particleUpdateScript [
   &ParticleState()
   &SkeletalMeshLocation(
    Bone / Socket Index=&SpawnGroup()
    MaskSpawnGroup=false
    Mesh Sampling Type=::NewEnumerator5
    Sampled Position Offset=&MakeVector(
     X=0
     Y=0
     Z=&FloatFromCurve(FloatCurve=DataInterface() {propertyValues {Curve {keys [{interpMode=RCIM_Cubic, leaveTangent=-0.4441, tangentMode=RCTM_Break, time=0, value=2}, <at 1 10>]}}})
    )
    Skeletal Mesh=$User.Skeletal Mesh
    Socket Sampling Mode=::NewEnumerator5
   )
   /SetVariables(
    Particles.SpriteAlignment=&RotateVectorByQuaternion(Vector To Rotate=<1 0 0>, Quaternion=$StackContext.SkeletalMeshLocation.SampledRotation)
    Particles.SpriteFacing=&RotateVectorByQuaternion(Vector To Rotate=<0 0 1>, Quaternion=$StackContext.SkeletalMeshLocation.SampledRotation)
   ) {isSetParameters=true, module=SetVariables_#4}
   &CameraOffset(Camera Offset Amount=5)
   &DynamicMaterialParameters(
    Index 0 Param 1=&MakeFloatFromVector(Vector=$Particles.Scale, Channel=::NewEnumerator0)
    Index 0 Param 2=&FloatFromCurve(FloatCurve=DataInterface() {propertyValues {Curve {keys [<at 0 1>,<at 1 2>]}}}, Scale Curve=3)
    Module.Index0Param0Write=true
    Module.Index0Param1Write=true
    Module.Index0Param2Write=true
    Module.Index0Param3Write=true
    Module.Index1Param0Write=true
    Module.Index1Param1Write=true
    Module.Index1Param2Write=true
    Module.Index1Param3Write=true
    Write Parameter Index 1=true
   )
   &ScaleMeshSize(Scale Factor=&VectorFromFloat(Value=&FloatFromCurve(FloatCurve=DataInterface() {propertyValues {Curve {keys [
          {interpMode=RCIM_Cubic, leaveTangent=3.5342, tangentMode=RCTM_Break, time=0, value=0.5}
          {arriveTangent=1.3693, interpMode=RCIM_Cubic, tangentMode=RCTM_Break, time=1, value=2}
         ]}}})))
   &ScaleColor(
    Scale Alpha=&MakeCustomFloatFromBool(
     Bool=$Particles.bool
     True Float=&FloatFromCurve(
      FloatCurve=DataInterface() {propertyValues {Curve {keys [
          {interpMode=RCIM_Cubic, leaveTangent=-3.4056, tangentMode=RCTM_Break, time=0, value=1}
          {arriveTangent=-0.0706, interpMode=RCIM_Cubic, tangentMode=RCTM_Break, time=1, value=0}
         ]}}}
      Scale Curve=2
     )
     False Float=&FloatFromCurve(
      FloatCurve=DataInterface() {propertyValues {Curve {keys [
          {interpMode=RCIM_Cubic, leaveTangent=-0.0192, tangentMode=RCTM_Break, time=0, value=1}
          {arriveTangent=-6.6328, interpMode=RCIM_Cubic, leaveTangent=-0.6773, tangentMode=RCTM_Break, time=0.2, value=0.2}
          {arriveTangent=-0.0456, interpMode=RCIM_Cubic, tangentMode=RCTM_Break, time=1, value=0}
         ]}}}
      Scale Curve=2
     )
    )
    Scale RGB=&VectorFromFloat(Value=&FloatFromCurve(FloatCurve=DataInterface() {propertyValues {Curve {keys [<at 0 1>, <at 0.1 3.5>, {arriveTangent=-2.6622, leaveTangent=-2.6622, time=0.2, value=2}, <at 1 0.45>]}}}))
    ScaleA=true
    ScaleRGB=true
   )
   /SetVariables(Particles.MeshOrientation=&MultiplyQuaternion(
     Quaternion A=$Particles.MeshOrientation
     Quaternion B=&MakeQuatFromAxisAngle(Axis=<1 0 0>, AngleInDegrees=0)
    )) {isSetParameters=true}
   &ScaleColor(
    Scale RGB=&VectorFromFloat(Value=&FNTimeOfDayScalar(Night Scale=1, Morning Scale=1, Evening Scale=1))
    ScaleA=true
    ScaleRGB=true
   ) {module=ScaleColor001}
  ]
  rendererClasses [&NiagaraMeshRendererProperties]
  renderers [{rendererClass=&NiagaraMeshRendererProperties, rendererIndex=0}]
 }
 Glow {
  emitterUpdateScript [
   &EmitterState(
    Inactive Response=::NewEnumerator1
    Life Cycle Mode=::NewEnumerator1
    Loop Behavior=::NewEnumerator1
    Loop Duration Mode=::NewEnumerator1
    UseLoopDelay=false
   )
   &SpawnBurst_Instantaneous(Spawn Count=4, Use Spawn Probability=false, UseLoopCountLimit=false)
  ]
  particleSpawnScript [
   &self:4Group4Particle(SpawnParticleCount=4) {module=4Group4Particle}
   &InitializeParticle(
    Color=<rgba 0.2 0.2 0.2 0.2>
    Lifetime=5
    Mesh Scale Mode=::NewEnumerator3
    Mesh Uniform Scale=0.1
    Sprite Size Mode=::NewEnumerator3
    Uniform Sprite Size=12
    UsePositionOffset=false
    Write Material Random=true
   )
  ]
  particleUpdateScript [
   &ParticleState(Kill Particles When Lifetime Has Elapsed=false, Let Infinitely Lived Particles Die When Emitter Deactivates=true)
   &SkeletalMeshLocation(
    Bone / Socket Index=&SelectIntFromArray(Direct Array Index=$Particles.SeqID, Int Selection Array=DataInterface() {propertyValues {IntData [0,1,3,4]}})
    MaskSpawnGroup=false
    Mesh Sampling Type=::NewEnumerator5
    Skeletal Mesh=$User.Skeletal Mesh
    Socket Sampling Mode=::NewEnumerator5
   )
   /SetVariables(
    Particles.SpriteAlignment=&RotateVectorByQuaternion(Vector To Rotate=<1 0 0>, Quaternion=$StackContext.SkeletalMeshLocation.SampledRotation)
    Particles.SpriteFacing=&RotateVectorByQuaternion(Vector To Rotate=<0 0 1>, Quaternion=$StackContext.SkeletalMeshLocation.SampledRotation)
   ) {isSetParameters=true, module=SetVariables_#4}
   &CameraOffset(Camera Offset Amount=1)
  ]
  rendererClasses [&NiagaraSpriteRendererProperties]
  renderers [{rendererClass=&NiagaraSpriteRendererProperties}]
 }
 Glow001 {
  emitterUpdateScript [
   &EmitterState(
    Inactive Response=::NewEnumerator1
    Life Cycle Mode=::NewEnumerator1
    Loop Behavior=::NewEnumerator1
    Loop Duration Mode=::NewEnumerator1
    UseLoopDelay=false
   )
   &SpawnBurst_Instantaneous(Spawn Count=2, Use Spawn Probability=false, UseLoopCountLimit=false)
  ]
  particleSpawnScript [
   &self:4Group4Particle(GroupCount=1, MaxParticle=2, SpawnParticleCount=2) {module=4Group4Particle}
   &InitializeParticle(
    Color=<rgba 0 0.125 0.3 1>
    Lifetime=5
    Mesh Scale Mode=::NewEnumerator3
    Mesh Uniform Scale=0.1
    Sprite Size Mode=::NewEnumerator3
    Uniform Sprite Size=30
    UsePositionOffset=false
    Write Material Random=true
   )
  ]
  particleUpdateScript [
   &ParticleState(Kill Particles When Lifetime Has Elapsed=false, Let Infinitely Lived Particles Die When Emitter Deactivates=true)
   &SkeletalMeshLocation(
    Bone / Socket Index=&SelectIntFromArray(Direct Array Index=$Particles.SeqID, Int Selection Array=DataInterface() {propertyValues {IntData [2,5,0,0]}})
    MaskSpawnGroup=false
    Mesh Sampling Type=::NewEnumerator5
    Sampled Position Offset=<0 0 -4>
    Skeletal Mesh=$User.Skeletal Mesh
    Socket Sampling Mode=::NewEnumerator5
   )
   /SetVariables(
    Particles.SpriteAlignment=&RotateVectorByQuaternion(Vector To Rotate=<1 0 0>, Quaternion=$StackContext.SkeletalMeshLocation.SampledRotation)
    Particles.SpriteFacing=&RotateVectorByQuaternion(Vector To Rotate=<0 0 1>, Quaternion=$StackContext.SkeletalMeshLocation.SampledRotation)
   ) {isSetParameters=true, module=SetVariables_#4}
   &CameraOffset(Camera Offset Amount=20)
   /SetVariables(
    Particles.Color2=&HueShiftLinearColor(Hue=-0.2, Saturation=1, Value=0.15, Linear Color=$Particles.Color)
    Particles.SpriteSize2=&Multiply_Vector2D_ByFloat(Float=3, Vector2D=$Particles.SpriteSize)
   ) {isSetParameters=true}
   &self:-----------Reactive----------() {module=-----------Reactive----------}
   &ScaleColor(
    Scale Alpha=&FloatFromCurve(
     FloatCurve=DataInterface() {propertyValues {Curve {keys [
         {interpMode=RCIM_Cubic, leaveTangent=0.2728, tangentMode=RCTM_Break, time=0, value=1}
         {arriveTangent=0.9757, interpMode=RCIM_Cubic, leaveTangent=-0.1879, tangentMode=RCTM_Break, time=0.9, value=1.5}
         {arriveTangent=-4.7154, tangentMode=RCTM_User, time=1, value=1}
        ]}}}
     CurveIndex=$System.ReactiveValue
    )
    Scale RGB=&VectorFromFloat(Value=&Add_Float(
      A=&FloatFromCurve(
       FloatCurve=DataInterface() {propertyValues {Curve {keys [
           {interpMode=RCIM_Cubic, tangentMode=RCTM_Break, time=0, value=0}
           {arriveTangent=17.2895, interpMode=RCIM_Cubic, leaveTangent=-6.938, tangentMode=RCTM_Break, time=0.9, value=5}
           {arriveTangent=-82.1567, tangentMode=RCTM_User, time=1, value=0}
          ]}}}
       CurveIndex=$System.ReactiveValue
       Scale Curve=&FN_FrontEndFloatModifier(Gameplay=1, Front End=0.7)
      )
      B=1
     ))
    ScaleA=true
    ScaleRGB=true
   ) {module=ScaleColor001}
   &ScaleSpriteSize(
    Uniform Curve Index=$System.ReactiveValue
    Uniform Curve Sprite Scale=DataInterface() {propertyValues {Curve {keys [{tangentMode=RCTM_User, time=0, value=1}, {tangentMode=RCTM_User, time=0.9, value=1.2}, <at 1 1>]}}}
   )
   &ScaleColor(
    Scale RGB=&VectorFromFloat(Value=&FNTimeOfDayScalar(Night Scale=1, Morning Scale=1, Evening Scale=1))
    ScaleA=true
    ScaleRGB=true
   )
  ]
  rendererClasses [&NiagaraSpriteRendererProperties]
  renderers [
   {rendererClass=&NiagaraSpriteRendererProperties}
   {rendererClass=&NiagaraSpriteRendererProperties, rendererIndex=1}
  ]
 }
 PostLine {
  emitterUpdateScript [
   &EmitterState(
    Inactive Response=::NewEnumerator1
    Life Cycle Mode=::NewEnumerator1
    Loop Behavior=::NewEnumerator1
    Loop Duration Mode=::NewEnumerator1
    UseLoopDelay=false
   )
   /SetVariables(Emitter.TheSpawnRate:NiagaraFloat=&FN_PickFloatDependOnPlatform(
     TargetPlatformActive=&FloatFromCurve(FloatCurve=DataInterface() {propertyValues {Curve {keys [<at 0 1>,<at 0.1 0>,<at 0.45 0>,<at 1 8>]}}}, CurveIndex=$System.ReactiveValue)
     TargetPlatformNoActive=&FloatFromCurve(FloatCurve=DataInterface() {propertyValues {Curve {keys [<at 0 1>,<at 0.1 0>,<at 0.45 0>,<at 1 4>]}}}, CurveIndex=$System.ReactiveValue)
    )) {isSetParameters=true}
   &SpawnRate(SpawnGroup=2, SpawnRate=$Emitter.TheSpawnRate, Use Spawn Probability=false)
   &SpawnRate(SpawnGroup=5, SpawnRate=$Emitter.TheSpawnRate, Use Spawn Probability=false) {module=SpawnRate001}
  ]
  particleSpawnScript [
   &InitializeParticle(
    Color=<rgba 0.2 0.36 1 0.2>
    Lifetime=0.6
    Mesh Scale Mode=::NewEnumerator3
    Mesh Uniform Scale=0.03
    Sprite Size Mode=::NewEnumerator3
    Uniform Sprite Size=50
    UsePositionOffset=false
    Write Material Random=true
   )
   /SetVariables(Particles.VisibilityTag=&MakeCustomIntFromBool(Bool=$User.bIsFrontEndPreview, True INT=1, False INT=0)) {isSetParameters=true}
  ]
  particleUpdateScript [
   &ParticleState()
   &SkeletalMeshLocation(
    Bone / Socket Index=&SpawnGroup()
    MaskSpawnGroup=false
    Mesh Sampling Type=::NewEnumerator5
    Skeletal Mesh=$User.Skeletal Mesh
    Socket Sampling Mode=::NewEnumerator5
   )
   /SetVariables(
    Particles.SpriteAlignment=&RotateVectorByQuaternion(Vector To Rotate=<1 0 0>, Quaternion=$StackContext.SkeletalMeshLocation.SampledRotation)
    Particles.SpriteFacing=&RotateVectorByQuaternion(Vector To Rotate=<0 0 1>, Quaternion=$StackContext.SkeletalMeshLocation.SampledRotation)
   ) {enabled=false, isSetParameters=true, module=SetVariables_#4}
   &SkeletalMeshLocation(
    Bone / Socket Index=8
    Bone Sampling Mode=::NewEnumerator4
    MaskSpawnGroup=false
    Orientation Sampling=::NewEnumerator1
    Position Sampling=::NewEnumerator2
    Skeletal Mesh=$User.Skeletal Mesh
   ) {module=SkeletalMeshLocation001}
   &CameraOffset(Camera Offset Amount=20)
   /SetVariables(Particles.OffsetPosition:NiagaraPosition=&AddVectorToPosition(
     Position=$Particles.Position
     Vector=&RotateVectorByQuaternion(Vector To Rotate=<0 0 -5>, Quaternion=$Particles.SkeletalMeshLocation.SampledRotation)
    )) {isSetParameters=true, module=SetVariables_#5}
   &DynamicMaterialParameters(
    Index 0 Param 1=&MakeFloatFromVector(Vector=Linked() {linkedVariable {name=Particles.Position, type=NiagaraPosition}}, Channel=::NewEnumerator0)
    Index 0 Param 2=&MakeFloatFromVector(Vector=Linked() {linkedVariable {name=Particles.Position, type=NiagaraPosition}}, Channel=::NewEnumerator1)
    Index 0 Param 3=&MakeFloatFromVector(Vector=Linked() {linkedVariable {name=Particles.Position, type=NiagaraPosition}}, Channel=::NewEnumerator2)
    Index 0 Param 4=&Lerp_Float(
     A=&FloatFromCurve(FloatCurve=DataInterface() {propertyValues {Curve {keys [
          {arriveTangent=-0.2971, interpMode=RCIM_Cubic, leaveTangent=-0.2971, tangentMode=RCTM_User, time=0, value=1.1}
          {arriveTangent=-0.1318, interpMode=RCIM_Cubic, leaveTangent=-0.1318, tangentMode=RCTM_User, time=1, value=0.8}
         ]}}})
     B=&FloatFromCurve(FloatCurve=DataInterface() {propertyValues {Curve {keys [
          {arriveTangent=-0.2971, interpMode=RCIM_Cubic, leaveTangent=-0.2971, tangentMode=RCTM_User, time=0, value=1}
          {arriveTangent=-0.1318, interpMode=RCIM_Cubic, leaveTangent=-0.1318, tangentMode=RCTM_User, time=1, value=0.6}
         ]}}})
     Alpha=&FloatFromCurve(FloatCurve=DataInterface() {propertyValues {Curve {keys [<at 0.5 0>,<at 1 1>]}}}, CurveIndex=$System.ReactiveValue)
    )
    Index 1 Param 1=&MakeFloatFromVector(Vector=$Particles.SkeletalMeshLocation001.SampledPosition, Channel=::NewEnumerator0)
    Index 1 Param 2=&MakeFloatFromVector(Vector=$Particles.SkeletalMeshLocation001.SampledPosition, Channel=::NewEnumerator1)
    Index 1 Param 3=&MakeFloatFromVector(Vector=$Particles.SkeletalMeshLocation001.SampledPosition, Channel=::NewEnumerator2)
    Index 1 Param 4=&FloatFromCurve(
     FloatCurve=DataInterface() {propertyValues {Curve {keys [
         {arriveTangent=0.1333, interpMode=RCIM_Cubic, leaveTangent=0.1333, tangentMode=RCTM_User, time=0, value=0}
         {interpMode=RCIM_Cubic, leaveTangent=-0.5523, tangentMode=RCTM_Break, time=0.3, value=0.8}
         {arriveTangent=-0.6776, interpMode=RCIM_Cubic, leaveTangent=-0.6776, tangentMode=RCTM_User, time=1, value=0.2}
        ]}}}
     Scale Curve=2
    )
    Index 2 Param 1=&FloatFromCurve(FloatCurve=DataInterface() {propertyValues {Curve {keys [<at 0 0>,<at 1 1>]}}}, Scale Curve=30)
    Module.Index0Param0Write=true
    Module.Index0Param1Write=true
    Module.Index0Param2Write=true
    Module.Index0Param3Write=true
    Module.Index1Param0Write=true
    Module.Index1Param1Write=true
    Module.Index1Param2Write=true
    Module.Index1Param3Write=true
    Module.Index2Param0Write=true
    Module.Index2Param1Write=true
    Module.Index2Param2Write=true
    Module.Index2Param3Write=true
    Write Parameter Index 1=true
    Write Parameter Index 2=true
   )
   &ScaleColor(
    Scale Alpha=&FloatFromCurve(
     FloatCurve=DataInterface() {propertyValues {Curve {keys [
         {arriveTangent=19.3619, interpMode=RCIM_Cubic, leaveTangent=19.3619, tangentMode=RCTM_User, time=0, value=0}
         {arriveTangent=-2.6364, interpMode=RCIM_Cubic, leaveTangent=-2.6364, tangentMode=RCTM_User, time=0.15, value=0.6}
         {arriveTangent=-0.6911, interpMode=RCIM_Cubic, leaveTangent=-0.6911, tangentMode=RCTM_User, time=1, value=0}
        ]}}}
     Scale Curve=1.2
    )
    Scale RGB=&VectorFromFloat(Value=&FloatFromCurve(FloatCurve=DataInterface() {propertyValues {Curve {keys [
          {arriveTangent=0.168, interpMode=RCIM_Cubic, leaveTangent=0.168, tangentMode=RCTM_User, time=0.2, value=1.5}
          {arriveTangent=-0.1012, interpMode=RCIM_Cubic, leaveTangent=-0.1012, tangentMode=RCTM_User, time=0.4, value=1}
         ]}}}))
    ScaleA=true
    ScaleRGB=true
   )
   &self:-----------Reactive----------() {module=-----------Reactive----------}
   &ScaleColor(
    Scale Alpha=&FloatFromCurve(
     FloatCurve=DataInterface() {propertyValues {Curve {keys [
         {interpMode=RCIM_Cubic, leaveTangent=0.431, tangentMode=RCTM_Break, time=0, value=1.3}
         {arriveTangent=3.1641, interpMode=RCIM_Cubic, leaveTangent=-6.0307, tangentMode=RCTM_Break, time=0.9, value=2}
         <at 1 1.3>
        ]}}}
     CurveIndex=$System.ReactiveValue
    )
    Scale RGB=&VectorFromFloat(Value=&MakeCustomFloatFromBool(
      Bool=$User.bIsFrontEndPreview
      True Float=&FloatFromCurve(
       FloatCurve=DataInterface() {propertyValues {Curve {keys [
           {interpMode=RCIM_Cubic, tangentMode=RCTM_Break, time=0, value=2.5}
           {arriveTangent=30.383, interpMode=RCIM_Cubic, leaveTangent=-248.8788, tangentMode=RCTM_Break, time=0.9, value=13}
           <at 1 2.5>
          ]}}}
       CurveIndex=$System.ReactiveValue
      )
      False Float=&FloatFromCurve(
       FloatCurve=DataInterface() {propertyValues {Curve {keys [
           {interpMode=RCIM_Cubic, tangentMode=RCTM_Break, time=0, value=1.3}
           {arriveTangent=43.2382, interpMode=RCIM_Cubic, leaveTangent=-248.8788, tangentMode=RCTM_Break, time=0.9, value=13}
           <at 1 1.3>
          ]}}}
       CurveIndex=$System.ReactiveValue
      )
     ))
    ScaleA=true
    ScaleRGB=true
   ) {module=ScaleColor001}
   &ScaleSpriteSize(
    Uniform Curve Index=$System.ReactiveValue
    Uniform Curve Sprite Scale=DataInterface() {propertyValues {Curve {keys [{tangentMode=RCTM_User, time=0, value=1}, {tangentMode=RCTM_User, time=0.9, value=2}, <at 1 1>]}}}
   )
   &KillParticles(Kill Particles=$System.DoOnce.Execute)
   &ScaleColor(
    Scale RGB=&VectorFromFloat(Value=&MakeCustomFloatFromBool(
      Bool=$User.bIsFrontEndPreview
      True Float=1
      False Float=&Add_Float(
       A=&FloatFromCurve(
        FloatCurve=DataInterface() {propertyValues {Curve {keys [{interpMode=RCIM_Cubic, leaveTangent=2.0965, tangentMode=RCTM_Break, time=0, value=0}, <at 1 1>]}}}
        CurveIndex=$NPC.FortniteNPC.EventAudio_Amplitude
        Scale Curve=7
       )
       B=1
      )
     ))
    ScaleA=true
    ScaleRGB=true
   ) {module=ScaleColor002}
  ]
  rendererClasses [&NiagaraSpriteRendererProperties, &NiagaraMeshRendererProperties]
  renderers [
   {rendererClass=&NiagaraSpriteRendererProperties}
   {rendererClass=&NiagaraMeshRendererProperties, rendererIndex=1}
   {rendererClass=&NiagaraSpriteRendererProperties, rendererIndex=2}
  ]
 }
 PostLine001 {
  emitterUpdateScript [
   &EmitterState(
    Inactive Response=::NewEnumerator1
    Life Cycle Mode=::NewEnumerator1
    Loop Behavior=::NewEnumerator1
    Loop Duration Mode=::NewEnumerator1
    UseLoopDelay=false
   )
   &SpawnPerFrame(Spawn:NiagaraBool=$System.DoOnce.Execute, Spawn Group=2, Use Spawn Probability=false)
   &SpawnPerFrame(Spawn:NiagaraBool=$System.DoOnce.Execute, Spawn Group=5, Use Spawn Probability=false) {module=SpawnPerFrame001}
   &SpawnRate(SpawnGroup=2, SpawnRate=$Emitter.TheSpawnRate, Use Spawn Probability=false) {enabled=false}
   &SpawnRate(SpawnGroup=5, SpawnRate=$Emitter.TheSpawnRate, Use Spawn Probability=false) {enabled=false, module=SpawnRate001}
  ]
  particleSpawnScript [
   &InitializeParticle(
    Color=<rgba 10 10 10 2>
    Lifetime=0.4
    Mesh Scale Mode=::NewEnumerator3
    Mesh Uniform Scale=0.03
    Sprite Size Mode=::NewEnumerator3
    Uniform Sprite Size=100
    UsePositionOffset=false
    Write Material Random=true
   )
   /SetVariables(Particles.VisibilityTag=&MakeCustomIntFromBool(Bool=$User.bIsFrontEndPreview, True INT=1, False INT=0)) {isSetParameters=true}
  ]
  particleUpdateScript [
   &ParticleState()
   &SkeletalMeshLocation(
    Bone / Socket Index=&SpawnGroup()
    MaskSpawnGroup=false
    Mesh Sampling Type=::NewEnumerator5
    Skeletal Mesh=$User.Skeletal Mesh
    Socket Sampling Mode=::NewEnumerator5
   )
   /SetVariables(
    Particles.SpriteAlignment=&RotateVectorByQuaternion(Vector To Rotate=<1 0 0>, Quaternion=$StackContext.SkeletalMeshLocation.SampledRotation)
    Particles.SpriteFacing=&RotateVectorByQuaternion(Vector To Rotate=<0 0 1>, Quaternion=$StackContext.SkeletalMeshLocation.SampledRotation)
   ) {enabled=false, isSetParameters=true, module=SetVariables_#4}
   &SkeletalMeshLocation(
    Bone / Socket Index=8
    Bone Sampling Mode=::NewEnumerator4
    MaskSpawnGroup=false
    Orientation Sampling=::NewEnumerator1
    Position Sampling=::NewEnumerator2
    Skeletal Mesh=$User.Skeletal Mesh
   ) {module=SkeletalMeshLocation001}
   &CameraOffset(Camera Offset Amount=20)
   /SetVariables(Particles.OffsetPosition:NiagaraPosition=&AddVectorToPosition(
     Position=$Particles.Position
     Vector=&RotateVectorByQuaternion(Vector To Rotate=<0 0 -5>, Quaternion=$Particles.SkeletalMeshLocation.SampledRotation)
    )) {isSetParameters=true, module=SetVariables_#5}
   &DynamicMaterialParameters(
    Index 0 Param 1=&MakeFloatFromVector(Vector=Linked() {linkedVariable {name=Particles.Position, type=NiagaraPosition}}, Channel=::NewEnumerator0)
    Index 0 Param 2=&MakeFloatFromVector(Vector=Linked() {linkedVariable {name=Particles.Position, type=NiagaraPosition}}, Channel=::NewEnumerator1)
    Index 0 Param 3=&MakeFloatFromVector(Vector=Linked() {linkedVariable {name=Particles.Position, type=NiagaraPosition}}, Channel=::NewEnumerator2)
    Index 0 Param 4=&FloatFromCurve(FloatCurve=DataInterface() {propertyValues {Curve {keys [
         {arriveTangent=-0.2971, interpMode=RCIM_Cubic, leaveTangent=-0.2971, tangentMode=RCTM_User, time=0, value=1.1}
         {arriveTangent=-0.1318, interpMode=RCIM_Cubic, leaveTangent=-0.1318, tangentMode=RCTM_User, time=1, value=0.8}
        ]}}})
    Index 1 Param 1=&MakeFloatFromVector(Vector=$Particles.SkeletalMeshLocation001.SampledPosition, Channel=::NewEnumerator0)
    Index 1 Param 2=&MakeFloatFromVector(Vector=$Particles.SkeletalMeshLocation001.SampledPosition, Channel=::NewEnumerator1)
    Index 1 Param 3=&MakeFloatFromVector(Vector=$Particles.SkeletalMeshLocation001.SampledPosition, Channel=::NewEnumerator2)
    Index 1 Param 4=&FloatFromCurve(
     FloatCurve=DataInterface() {propertyValues {Curve {keys [
         {arriveTangent=0.1333, interpMode=RCIM_Cubic, leaveTangent=0.1333, tangentMode=RCTM_User, time=0, value=0}
         {interpMode=RCIM_Cubic, leaveTangent=-0.5523, tangentMode=RCTM_Break, time=0.3, value=1}
         {arriveTangent=-0.6776, interpMode=RCIM_Cubic, leaveTangent=-0.6776, tangentMode=RCTM_User, time=1, value=0.3}
        ]}}}
     Scale Curve=2
    )
    Index 2 Param 1=50
    Module.Index0Param0Write=true
    Module.Index0Param1Write=true
    Module.Index0Param2Write=true
    Module.Index0Param3Write=true
    Module.Index1Param0Write=true
    Module.Index1Param1Write=true
    Module.Index1Param2Write=true
    Module.Index1Param3Write=true
    Module.Index2Param0Write=true
    Module.Index2Param1Write=true
    Module.Index2Param2Write=true
    Module.Index2Param3Write=true
    Write Parameter Index 1=true
    Write Parameter Index 2=true
   )
   &ScaleColor(
    Scale Alpha=&FloatFromCurve(FloatCurve=DataInterface() {propertyValues {Curve {keys [
         {arriveTangent=19.3619, interpMode=RCIM_Cubic, leaveTangent=19.3619, tangentMode=RCTM_User, time=0, value=0}
         {arriveTangent=-2.6364, interpMode=RCIM_Cubic, leaveTangent=-2.6364, tangentMode=RCTM_User, time=0.1, value=0.65}
         <at 0.138 0.0903>
         {arriveTangent=-0.222, interpMode=RCIM_Cubic, leaveTangent=-0.222, tangentMode=RCTM_User, time=1, value=0}
        ]}}})
    Scale RGB=&VectorFromFloat(Value=&FloatFromCurve(FloatCurve=DataInterface() {propertyValues {Curve {keys [
          {arriveTangent=0.168, interpMode=RCIM_Cubic, leaveTangent=0.168, tangentMode=RCTM_User, time=0, value=1.5}
          {arriveTangent=-0.1012, interpMode=RCIM_Cubic, leaveTangent=-0.1012, tangentMode=RCTM_User, time=0.2, value=0.1}
         ]}}}))
    ScaleA=true
    ScaleRGB=true
   )
  ]
  rendererClasses [&NiagaraSpriteRendererProperties, &NiagaraMeshRendererProperties]
  renderers [
   {rendererClass=&NiagaraSpriteRendererProperties}
   {rendererClass=&NiagaraMeshRendererProperties, rendererIndex=1}
   {rendererClass=&NiagaraSpriteRendererProperties, rendererIndex=2}
  ]
 }
 PostWave {
  emitterUpdateScript [
   &EmitterState(
    Inactive Response=::NewEnumerator1
    Life Cycle Mode=::NewEnumerator1
    Loop Behavior=::NewEnumerator1
    Loop Duration Mode=::NewEnumerator1
    UseLoopDelay=false
   )
   /SetVariables(Emitter.TheSpawnRate:NiagaraFloat=&FN_PickFloatDependOnPlatform(
     TargetPlatformActive=&FloatFromCurve(FloatCurve=DataInterface() {propertyValues {Curve {keys [<at 0 1>,<at 0.1 0>,<at 0.45 0>,<at 1 8>]}}}, CurveIndex=$System.ReactiveValue)
     TargetPlatformNoActive=&FloatFromCurve(FloatCurve=DataInterface() {propertyValues {Curve {keys [<at 0 1>,<at 0.1 0>,<at 0.45 0>,<at 1 4>]}}}, CurveIndex=$System.ReactiveValue)
    )) {isSetParameters=true}
   &SpawnRate(SpawnGroup=2, SpawnRate=$Emitter.TheSpawnRate, Use Spawn Probability=false)
   &SpawnRate(SpawnGroup=5, SpawnRate=$Emitter.TheSpawnRate, Use Spawn Probability=false) {module=SpawnRate001}
  ]
  particleSpawnScript [
   &InitializeParticle(
    Color=<rgba 0.2 0.36 1 20>
    Lifetime=0.4
    Mesh Scale Mode=::NewEnumerator3
    Mesh Uniform Scale=0.03
    Sprite Size Mode=::NewEnumerator3
    Uniform Sprite Size=50
    UsePositionOffset=false
    Write Material Random=true
   )
   /SetVariables(Particles.VisibilityTag=&MakeCustomIntFromBool(Bool=$User.bIsFrontEndPreview, True INT=1, False INT=0)) {isSetParameters=true}
  ]
  particleUpdateScript [
   &ParticleState()
   &SkeletalMeshLocation(
    Bone / Socket Index=&SpawnGroup()
    MaskSpawnGroup=false
    Mesh Sampling Type=::NewEnumerator5
    Skeletal Mesh=$User.Skeletal Mesh
    Socket Sampling Mode=::NewEnumerator5
   )
   /SetVariables(
    Particles.SpriteAlignment=&RotateVectorByQuaternion(Vector To Rotate=<1 0 0>, Quaternion=$StackContext.SkeletalMeshLocation.SampledRotation)
    Particles.SpriteFacing=&RotateVectorByQuaternion(Vector To Rotate=<0 0 1>, Quaternion=$StackContext.SkeletalMeshLocation.SampledRotation)
   ) {enabled=false, isSetParameters=true, module=SetVariables_#4}
   &SkeletalMeshLocation(
    Bone / Socket Index=8
    Bone Sampling Mode=::NewEnumerator4
    MaskSpawnGroup=false
    Orientation Sampling=::NewEnumerator1
    Position Sampling=::NewEnumerator2
    Skeletal Mesh=$User.Skeletal Mesh
   ) {module=SkeletalMeshLocation001}
   &CameraOffset(Camera Offset Amount=20)
   /SetVariables(Particles.OffsetPosition:NiagaraPosition=&AddVectorToPosition(
     Position=$Particles.Position
     Vector=&RotateVectorByQuaternion(
      Vector To Rotate=&MakeVector(
       X=0
       Y=0
       Z=&FloatFromCurve(FloatCurve=DataInterface() {propertyValues {Curve {keys [<at 0 -5>,<at 1 -15>]}}}, CurveIndex=$System.ReactiveValue)
      )
      Quaternion=$Particles.SkeletalMeshLocation.SampledRotation
     )
    )) {isSetParameters=true, module=SetVariables_#5}
   &DynamicMaterialParameters(
    Index 0 Param 1=&MakeFloatFromVector(Vector=Linked() {linkedVariable {name=Particles.Position, type=NiagaraPosition}}, Channel=::NewEnumerator0)
    Index 0 Param 2=&MakeFloatFromVector(Vector=Linked() {linkedVariable {name=Particles.Position, type=NiagaraPosition}}, Channel=::NewEnumerator1)
    Index 0 Param 3=&MakeFloatFromVector(Vector=Linked() {linkedVariable {name=Particles.Position, type=NiagaraPosition}}, Channel=::NewEnumerator2)
    Index 1 Param 1=&MakeFloatFromVector(Vector=Linked() {linkedVariable {name=Particles.OffsetPosition, type=NiagaraPosition}}, Channel=::NewEnumerator0)
    Index 1 Param 2=&MakeFloatFromVector(Vector=Linked() {linkedVariable {name=Particles.OffsetPosition, type=NiagaraPosition}}, Channel=::NewEnumerator1)
    Index 1 Param 3=&MakeFloatFromVector(Vector=Linked() {linkedVariable {name=Particles.OffsetPosition, type=NiagaraPosition}}, Channel=::NewEnumerator2)
    Index 1 Param 4=&Lerp_Float(
     A=&FloatFromCurve(FloatCurve=DataInterface() {propertyValues {Curve {keys [
          {arriveTangent=-0.0762, interpMode=RCIM_Cubic, leaveTangent=-0.0762, tangentMode=RCTM_User, time=0, value=0.8}
          {arriveTangent=-0.2085, interpMode=RCIM_Cubic, leaveTangent=-0.2085, tangentMode=RCTM_User, time=1, value=0.3}
         ]}}})
     B=&FloatFromCurve(FloatCurve=DataInterface() {propertyValues {Curve {keys [
          {arriveTangent=-0.8006, interpMode=RCIM_Cubic, leaveTangent=-0.8006, tangentMode=RCTM_User, time=0, value=0.8}
          {arriveTangent=-0.3233, interpMode=RCIM_Cubic, leaveTangent=-0.3233, tangentMode=RCTM_User, time=1, value=0.3}
         ]}}})
     Alpha=$System.ReactiveValue
    )
    Module.Index0Param0Write=true
    Module.Index0Param1Write=true
    Module.Index0Param2Write=true
    Module.Index0Param3Write=true
    Module.Index1Param0Write=true
    Module.Index1Param1Write=true
    Module.Index1Param2Write=true
    Module.Index1Param3Write=true
    Write Parameter Index 1=true
   )
   &ScaleColor(
    Scale Alpha=&FloatFromCurve(FloatCurve=DataInterface() {propertyValues {Curve {keys [
         {arriveTangent=0.5632, interpMode=RCIM_Cubic, leaveTangent=0.5632, tangentMode=RCTM_User, time=0, value=0}
         {arriveTangent=0.168, interpMode=RCIM_Cubic, leaveTangent=0.168, tangentMode=RCTM_User, time=0.25, value=1}
         {arriveTangent=-0.1012, interpMode=RCIM_Cubic, leaveTangent=-0.1012, tangentMode=RCTM_User, time=1, value=0}
        ]}}})
    Scale RGB=&VectorFromFloat(Value=&FloatFromCurve(FloatCurve=DataInterface() {propertyValues {Curve {keys [
          {arriveTangent=0.168, interpMode=RCIM_Cubic, leaveTangent=0.168, tangentMode=RCTM_User, time=0.2, value=1.5}
          {arriveTangent=-0.1012, interpMode=RCIM_Cubic, leaveTangent=-0.1012, tangentMode=RCTM_User, time=0.4, value=1}
         ]}}}))
    ScaleA=true
    ScaleRGB=true
   )
   &self:-----------Reactive----------() {module=-----------Reactive----------}
   &ScaleColor(
    Scale Alpha=&FloatFromCurve(
     FloatCurve=DataInterface() {propertyValues {Curve {keys [
         {interpMode=RCIM_Cubic, leaveTangent=0.431, tangentMode=RCTM_Break, time=0, value=1}
         {arriveTangent=3.4786, interpMode=RCIM_Cubic, leaveTangent=-19.0618, tangentMode=RCTM_Break, time=0.9, value=2}
         <at 1 1>
        ]}}}
     CurveIndex=$System.ReactiveValue
    )
    Scale RGB=&VectorFromFloat(Value=&FloatFromCurve(
      FloatCurve=DataInterface() {propertyValues {Curve {keys [
          {interpMode=RCIM_Cubic, tangentMode=RCTM_Break, time=0, value=1}
          {arriveTangent=21.7562, interpMode=RCIM_Cubic, leaveTangent=-131.2767, tangentMode=RCTM_Break, time=0.9, value=6}
          <at 1 1>
         ]}}}
      CurveIndex=$System.ReactiveValue
     ))
    ScaleA=true
    ScaleRGB=true
   ) {module=ScaleColor001}
   &ScaleSpriteSize(
    Uniform Curve Index=$System.ReactiveValue
    Uniform Curve Sprite Scale=DataInterface() {propertyValues {Curve {keys [{tangentMode=RCTM_User, time=0, value=1}, {tangentMode=RCTM_User, time=0.9, value=1}, <at 1 1>]}}}
   )
   &KillParticles(Kill Particles=$System.DoOnce.Execute)
   &ScaleColor(
    Scale RGB=&VectorFromFloat(Value=&FNTimeOfDayScalar(Night Scale=1, Morning Scale=1, Evening Scale=1))
    ScaleA=true
    ScaleRGB=true
   ) {module=ScaleColor002}
  ]
  rendererClasses [&NiagaraSpriteRendererProperties, &NiagaraMeshRendererProperties]
  renderers [
   {rendererClass=&NiagaraSpriteRendererProperties}
   {rendererClass=&NiagaraMeshRendererProperties, rendererIndex=1}
   {rendererClass=&NiagaraSpriteRendererProperties, rendererIndex=2}
   {rendererClass=&NiagaraSpriteRendererProperties, rendererIndex=3}
  ]
 }
 PostWave001 {
  emitterUpdateScript [
   &EmitterState(
    Inactive Response=::NewEnumerator1
    Life Cycle Mode=::NewEnumerator1
    Loop Behavior=::NewEnumerator1
    Loop Duration Mode=::NewEnumerator1
    UseLoopDelay=false
   )
   &SpawnPerFrame(Spawn:NiagaraBool=$System.DoOnce.Execute, Spawn Group=2, Use Spawn Probability=false)
   &SpawnPerFrame(Spawn:NiagaraBool=$System.DoOnce.Execute, Spawn Group=5, Use Spawn Probability=false) {module=SpawnPerFrame001}
  ]
  particleSpawnScript [
   &InitializeParticle(
    Color=<rgba 1.5 3.75 15 5>
    Lifetime=0.3
    Mesh Scale Mode=::NewEnumerator3
    Mesh Uniform Scale=0.03
    Sprite Size Mode=::NewEnumerator3
    Uniform Sprite Size=109.0912
    UsePositionOffset=false
    Write Material Random=true
   )
   /SetVariables(Particles.VisibilityTag=&MakeCustomIntFromBool(Bool=$User.bIsFrontEndPreview, True INT=1, False INT=0)) {isSetParameters=true}
  ]
  particleUpdateScript [
   &ParticleState()
   &SkeletalMeshLocation(
    Bone / Socket Index=&SpawnGroup()
    MaskSpawnGroup=false
    Mesh Sampling Type=::NewEnumerator5
    Skeletal Mesh=$User.Skeletal Mesh
    Socket Sampling Mode=::NewEnumerator5
   )
   /SetVariables(
    Particles.SpriteAlignment=&RotateVectorByQuaternion(Vector To Rotate=<1 0 0>, Quaternion=$StackContext.SkeletalMeshLocation.SampledRotation)
    Particles.SpriteFacing=&RotateVectorByQuaternion(Vector To Rotate=<0 0 1>, Quaternion=$StackContext.SkeletalMeshLocation.SampledRotation)
   ) {enabled=false, isSetParameters=true, module=SetVariables_#4}
   &SkeletalMeshLocation(
    Bone / Socket Index=8
    Bone Sampling Mode=::NewEnumerator4
    MaskSpawnGroup=false
    Orientation Sampling=::NewEnumerator1
    Position Sampling=::NewEnumerator2
    Skeletal Mesh=$User.Skeletal Mesh
   ) {module=SkeletalMeshLocation001}
   &CameraOffset(Camera Offset Amount=20)
   /SetVariables(Particles.OffsetPosition:NiagaraPosition=&AddVectorToPosition(
     Position=$Particles.Position
     Vector=&RotateVectorByQuaternion(
      Vector To Rotate=&MakeVector(
       X=0
       Y=0
       Z=&FloatFromCurve(FloatCurve=DataInterface() {propertyValues {Curve {keys [<at 0 -5>,<at 1 -15>]}}}, CurveIndex=$System.ReactiveValue)
      )
      Quaternion=$Particles.SkeletalMeshLocation.SampledRotation
     )
    )) {isSetParameters=true, module=SetVariables_#5}
   &DynamicMaterialParameters(
    Index 0 Param 1=&MakeFloatFromVector(Vector=Linked() {linkedVariable {name=Particles.Position, type=NiagaraPosition}}, Channel=::NewEnumerator0)
    Index 0 Param 2=&MakeFloatFromVector(Vector=Linked() {linkedVariable {name=Particles.Position, type=NiagaraPosition}}, Channel=::NewEnumerator1)
    Index 0 Param 3=&MakeFloatFromVector(Vector=Linked() {linkedVariable {name=Particles.Position, type=NiagaraPosition}}, Channel=::NewEnumerator2)
    Index 0 Param 4=30
    Index 1 Param 1=&MakeFloatFromVector(Vector=$Particles.SkeletalMeshLocation001.SampledPosition, Channel=::NewEnumerator0)
    Index 1 Param 2=&MakeFloatFromVector(Vector=$Particles.SkeletalMeshLocation001.SampledPosition, Channel=::NewEnumerator1)
    Index 1 Param 3=&MakeFloatFromVector(Vector=$Particles.SkeletalMeshLocation001.SampledPosition, Channel=::NewEnumerator2)
    Index 1 Param 4=&FloatFromCurve(FloatCurve=DataInterface() {propertyValues {Curve {keys [
         {arriveTangent=-0.4679, interpMode=RCIM_Cubic, leaveTangent=-0.4679, tangentMode=RCTM_User, time=0, value=1}
         {arriveTangent=-0.1394, interpMode=RCIM_Cubic, leaveTangent=-0.1394, tangentMode=RCTM_User, time=1, value=0.5}
        ]}}})
    Module.Index0Param0Write=true
    Module.Index0Param1Write=true
    Module.Index0Param2Write=true
    Module.Index0Param3Write=true
    Module.Index1Param0Write=true
    Module.Index1Param1Write=true
    Module.Index1Param2Write=true
    Module.Index1Param3Write=true
    Write Parameter Index 1=true
   )
   &ScaleColor(
    Scale Alpha=&FloatFromCurve(FloatCurve=DataInterface() {propertyValues {Curve {keys [
         {arriveTangent=0.5632, interpMode=RCIM_Cubic, leaveTangent=0.5632, tangentMode=RCTM_User, time=0, value=0}
         {arriveTangent=0.168, interpMode=RCIM_Cubic, leaveTangent=0.168, tangentMode=RCTM_User, time=0.25, value=1}
         {arriveTangent=-0.1012, interpMode=RCIM_Cubic, leaveTangent=-0.1012, tangentMode=RCTM_User, time=1, value=0}
        ]}}})
    Scale RGB=&VectorFromFloat(Value=&FloatFromCurve(FloatCurve=DataInterface() {propertyValues {Curve {keys [
          {arriveTangent=0.168, interpMode=RCIM_Cubic, leaveTangent=0.168, tangentMode=RCTM_User, time=0.2, value=1.5}
          {arriveTangent=-0.1012, interpMode=RCIM_Cubic, leaveTangent=-0.1012, tangentMode=RCTM_User, time=0.4, value=1}
         ]}}}))
    ScaleA=true
    ScaleRGB=true
   )
  ]
  rendererClasses [&NiagaraSpriteRendererProperties, &NiagaraMeshRendererProperties]
  renderers [
   {rendererClass=&NiagaraSpriteRendererProperties}
   {rendererClass=&NiagaraMeshRendererProperties, rendererIndex=1}
   {rendererClass=&NiagaraSpriteRendererProperties, rendererIndex=2}
  ]
 }
 Ring {
  emitterUpdateScript [
   &EmitterState(
    Inactive Response=::NewEnumerator1
    Life Cycle Mode=::NewEnumerator1
    Loop Behavior=::NewEnumerator1
    Loop Duration Mode=::NewEnumerator1
    UseLoopDelay=false
   )
   &SpawnBurst_Instantaneous(Spawn Count=20, Use Spawn Probability=false, UseLoopCountLimit=false)
  ]
  particleSpawnScript [
   &self:4Group4Particle(GroupCount=5, MaxParticle=20, SpawnParticleCount=20) {module=4Group4Particle}
   &InitializeParticle(
    Color=&SelectLinearColorFromArray(
     Array Sampling Mode=::NewEnumerator1
     Direct Array Index=$Particles.GroupID
     Color Selection Array=DataInterface() {propertyValues {ColorData [<rgba 0 0.091 0.2 2>, <rgba 0 0.6823 1.5 1>, <rgba 1.1813 0.225 1.5 0.8>, <rgba 1.0237 0.195 1.3 0.3>, <rgba 0 0.6823 1.5 0.3>]}}
     Random Seed=?
    )
    Lifetime=5
    Mesh Scale Mode=::NewEnumerator3
    Mesh Uniform Scale=&SelectFloatFromArray(Direct Array Index=$Particles.GroupID, Float Selection Array=DataInterface() {propertyValues {FloatData [0.03,0.018,0.03,0.03,0.018]}})
    Sprite Size Mode=::NewEnumerator3
    UsePositionOffset=false
    Write Material Random=true
   )
   /SetVariables(Particles.VisibilityTag=&SelectIntFromArray(Direct Array Index=$Particles.GroupID, Int Selection Array=DataInterface() {propertyValues {IntData [0,1,1,2,2]}})) {isSetParameters=true}
   &KillParticles(Kill Particles=&SetBoolByFloatComparison(
     A=&FN_PickFloatDependOnPlatform(
      TargetPlatformActive=-1
      TargetPlatformNoActive=&SelectFloatFromArray(Direct Array Index=$Particles.GroupID, Float Selection Array=DataInterface() {propertyValues {FloatData [-1,-1,-1,-1,1]}})
     )
     B=0
    ))
  ]
  particleUpdateScript [
   &ParticleState(Kill Particles When Lifetime Has Elapsed=false, Let Infinitely Lived Particles Die When Emitter Deactivates=true)
   &Timeline(
    Play Rate=&FloatFromCurve(FloatCurve=DataInterface() {propertyValues {Curve {keys [<at 0 1>,<at 0.9 2>,<at 1 1>]}}}, CurveIndex=$System.ReactiveValue)
    Timeline End=100000
   )
   &SkeletalMeshLocation(
    Bone / Socket Index=&SelectIntFromArray(Direct Array Index=$Particles.SeqID, Int Selection Array=DataInterface() {propertyValues {IntData [0,1,3,4]}})
    MaskSpawnGroup=false
    Mesh Sampling Type=::NewEnumerator5
    Sampled Position Offset=&SelectVectorFromArray(
     Direct Array Index=$Particles.GroupID
     Vector Selection Array=DataInterface() {propertyValues {FloatData [<0 0 1.75>, <0 0 1>, <0 0 2>, <0 0 2>, <0 0 1>]}}
    )
    Skeletal Mesh=$User.Skeletal Mesh
    Socket Sampling Mode=::NewEnumerator5
   )
   /SetVariables(
    Particles.SpriteAlignment=&RotateVectorByQuaternion(Vector To Rotate=<1 0 0>, Quaternion=$StackContext.SkeletalMeshLocation.SampledRotation)
    Particles.SpriteFacing=&RotateVectorByQuaternion(Vector To Rotate=<0 0 1>, Quaternion=$StackContext.SkeletalMeshLocation.SampledRotation)
   ) {isSetParameters=true, module=SetVariables_#4}
   &CameraOffset(Camera Offset Amount=5)
   &DynamicMaterialParameters(
    Index 0 Param 1=&MakeFloatFromVector(Vector=$Particles.Scale, Channel=::NewEnumerator0)
    Index 0 Param 2=5
    Index 0 Param 3=&Lerp_Float(
     A=0
     B=&Waveform(Global Input Position=-0.17, "Waveform [1]"=::NewEnumerator0, "[1] Phase Scale"=3, "[1] Frequency"=1, "[1] Amplitude Min/Max"=<0 0>)
     Alpha=&Power(a=$System.ReactiveValue, b=3)
    )
    Index 0 Param 4=&Waveform(Global Input Position=128, "Waveform [1]"=::NewEnumerator12, "[1] Phase Scale"=2, "[1] Frequency"=1, "[1] Amplitude Min/Max"=<-0.2 0.2>)
    Index 1 Param 1=3
    Index 1 Param 2=&SelectFloatFromArray(Direct Array Index=$Particles.GroupID, Float Selection Array=DataInterface() {propertyValues {FloatData [1,1,1,-0.75,0.2]}})
    Index 1 Param 3=&FloatFromCurve(
     FloatCurve=DataInterface() {propertyValues {Curve {keys [{interpMode=RCIM_Cubic, leaveTangent=0.7531, tangentMode=RCTM_Break, time=0, value=0}, <at 1 0.4>]}}}
     CurveIndex=$System.ReactiveValue
    )
    Index 1 Param 4=$Particles.Timeline.Time
    Index 2 Param 1=&FloatFromCurve(
     FloatCurve=DataInterface() {propertyValues {Curve {keys [{interpMode=RCIM_Cubic, tangentMode=RCTM_Break, time=0, value=1}, <at 1 3.5>]}}}
     CurveIndex=$System.ReactiveValue
    )
    Module.Index0Param0Write=true
    Module.Index0Param1Write=true
    Module.Index0Param2Write=true
    Module.Index0Param3Write=true
    Module.Index1Param0Write=true
    Module.Index1Param1Write=true
    Module.Index1Param2Write=true
    Module.Index1Param3Write=true
    Module.Index2Param0Write=true
    Module.Index2Param1Write=true
    Module.Index2Param2Write=true
    Module.Index2Param3Write=true
    Write Parameter Index 1=true
    Write Parameter Index 2=true
   )
   &ScaleMeshSize(Scale Factor=&VectorFromFloat(Value=&Lerp_Float(
      A=&Waveform(Global Input Position=-0.17, "Waveform [1]"=::NewEnumerator0, "[1] Frequency"=1, "[1] Amplitude Min/Max"=<1 1.2>)
      B=&Multiply_Float(
       A=&Waveform(Global Input Position=-0.17, "Waveform [1]"=::NewEnumerator0, "[1] Phase Scale"=5, "[1] Frequency"=1, "[1] Amplitude Min/Max"=<1.8 2.2>)
       B=&SelectFloatFromArray(
        Direct Array Index=$Particles.GroupID
        Direct Array Mode=::NewEnumerator1
        Float Selection Array=DataInterface() {propertyValues {FloatData [1,0.8,1,0.8,1]}}
       )
      )
      Alpha=&FloatFromCurve(
       FloatCurve=DataInterface() {propertyValues {Curve {keys [{interpMode=RCIM_Cubic, leaveTangent=1.8912, tangentMode=RCTM_Break, time=0, value=0}, <at 1 1>]}}}
       CurveIndex=$System.ReactiveValue
      )
     )))
   &ScaleColor(
    Scale RGB=&VectorFromFloat(Value=&Add_Float(
      A=&Multiply_Float(
       A=&FloatFromCurve(
        FloatCurve=DataInterface() {propertyValues {Curve {keys [
            {interpMode=RCIM_Cubic, tangentMode=RCTM_Break, time=0, value=0}
            {arriveTangent=3.7964, interpMode=RCIM_Cubic, leaveTangent=1.1533, tangentMode=RCTM_Break, time=0.9, value=3}
            <at 1 0>
           ]}}}
        CurveIndex=$System.ReactiveValue
        Scale Curve=&SelectFloatFromArray(Direct Array Index=$Particles.GroupID, Float Selection Array=DataInterface() {propertyValues {FloatData [0,1.6,0.75,2.5,2.5]}})
       )
       B=&FN_FrontEndFloatModifier(Gameplay=1, Front End=0.5)
      )
      B=1
     ))
    ScaleA=true
    ScaleRGB=true
   ) {module=ScaleColor001}
   &ScaleColor(
    Scale RGB=&VectorFromFloat(Value=&FN_FrontEndFloatModifier(Gameplay=&FNTimeOfDayScalar(Night Scale=1, Morning Scale=1, Evening Scale=1), Front End=0.9))
    ScaleA=true
    ScaleRGB=true
   )
  ]
  rendererClasses [&NiagaraMeshRendererProperties]
  renderers [
   {rendererClass=&NiagaraMeshRendererProperties, rendererIndex=0}
   {rendererClass=&NiagaraMeshRendererProperties, rendererIndex=1}
   {rendererClass=&NiagaraMeshRendererProperties, rendererIndex=2}
  ]
 }
 ShoulderLine {
  emitterUpdateScript [
   &EmitterState(
    Inactive Response=::NewEnumerator1
    Life Cycle Mode=::NewEnumerator1
    Loop Behavior=::NewEnumerator1
    Loop Duration Mode=::NewEnumerator1
    UseLoopDelay=false
   )
   &SpawnBurst_Instantaneous(Spawn Group=6, Use Spawn Probability=false, UseLoopCountLimit=false) {module=SpawnBurst_Instantaneous003}
   &SpawnBurst_Instantaneous(Spawn Group=7, Use Spawn Probability=false, UseLoopCountLimit=false) {module=SpawnBurst_Instantaneous002}
   &SpawnBurst_Instantaneous(Spawn Group=2, Use Spawn Probability=false, UseLoopCountLimit=false)
   &SpawnBurst_Instantaneous(Spawn Group=5, Use Spawn Probability=false, UseLoopCountLimit=false) {module=SpawnBurst_Instantaneous001}
  ]
  particleSpawnScript [
   /SetVariables(Particles.VisibilityTag=&MakeCustomIntFromBool(
     Bool=&Boolean_OrOperation(
      Bool1=&SetBoolByIntComparison(A=&SpawnGroup(), B=7, Comparison Type=::NewEnumerator4)
      Bool2=&SetBoolByIntComparison(A=&SpawnGroup(), B=6, Comparison Type=::NewEnumerator4)
     )
     True INT=1
     False INT=0
    )) {isSetParameters=true}
   &InitializeParticle(
    Color=&MakeCustomLinearColorFromBool(
     Bool=&SetBoolByIntComparison(A=$Particles.VisibilityTag, B=0, Comparison Type=::NewEnumerator0)
     True Color=&MakeLinearColor(
      R:NiagaraFloat=3
      G:NiagaraFloat=3
      B:NiagaraFloat=3
      A:NiagaraFloat=&FN_FrontEndFloatModifier(Gameplay=0.9, Front End=0.5)
     )
     False Color=&MakeLinearColor(
      R:NiagaraFloat=4
      G:NiagaraFloat=4
      B:NiagaraFloat=4
      A:NiagaraFloat=&FN_FrontEndFloatModifier(Gameplay=2, Front End=1)
     )
    )
    Lifetime=5
    Mesh Scale Mode=::NewEnumerator3
    Mesh Uniform Scale=0.15
    Sprite Size=<13 13>
    Sprite Size Mode=::NewEnumerator4
    UsePositionOffset=false
    Write Material Random=true
   )
   /SetVariables(Particles.PartHiddenCheckValue:NiagaraInt32=$Particles.VisibilityTag) {isSetParameters=true}
  ]
  particleUpdateScript [
   &ParticleState(Kill Particles When Lifetime Has Elapsed=false, Let Infinitely Lived Particles Die When Emitter Deactivates=true)
   &Timeline(
    Play Rate=&FloatFromCurve(FloatCurve=DataInterface() {propertyValues {Curve {keys [<at 0 1>,<at 1 8>]}}}, CurveIndex=$System.ReactiveValue)
    Timeline End=10000
   )
   &SkeletalMeshLocation(
    Bone / Socket Index=&SpawnGroup()
    MaskSpawnGroup=false
    Mesh Sampling Type=::NewEnumerator5
    Sampled Position Offset=&MakeCustomVectorFromBool(
     Bool=&SetBoolByIntComparison(A=$Particles.VisibilityTag, B=1, Comparison Type=::NewEnumerator4)
     True Vector=&MakeCustomVectorFromBool(
      Bool=/Niagara/DynamicInputs/Helpers/MaskBoolBySpawnGroup(Masked Bool:NiagaraBool=false, Passthrough Bool:NiagaraBool=true, Spawn Group Mask:NiagaraInt32=7)
      True Vector=<0 -3 0>
      False Vector=<0 3 0>
     )
     False Vector=<0 0 -2>
    )
    Skeletal Mesh=$User.Skeletal Mesh
    Socket Sampling Mode=::NewEnumerator5
   )
   &CameraOffset(Camera Offset Amount=5)
   &SkeletalMeshLocation(
    Bone / Socket Index=&MaskIntBySpawnGroup(Masked Int=1, Passthrough Int=3, Spawn Group Mask=2)
    MaskSpawnGroup=false
    Mesh Sampling Type=::NewEnumerator5
    Orientation Sampling=::NewEnumerator1
    Position Sampling=::NewEnumerator2
    Sampled Position Offset=&MaskVectorBySpawnGroup(Masked Vector=<0 0 0>, Passthrough Vector=<0 0 0>, Spawn Group Mask=2)
    Skeletal Mesh=$User.Skeletal Mesh
    Socket Sampling Mode=::NewEnumerator5
   ) {module=SkeletalMeshLocation001}
   &SkeletalMeshLocation(
    Bone / Socket Index=&MaskIntBySpawnGroup(Masked Int=0, Passthrough Int=4, Spawn Group Mask=2)
    MaskSpawnGroup=false
    Mesh Sampling Type=::NewEnumerator5
    Orientation Sampling=::NewEnumerator1
    Position Sampling=::NewEnumerator2
    Sampled Position Offset=&MaskVectorBySpawnGroup(Masked Vector=<0 0 0>, Passthrough Vector=<0 0 0>, Spawn Group Mask=2)
    Skeletal Mesh=$User.Skeletal Mesh
    Socket Sampling Mode=::NewEnumerator5
   ) {module=SkeletalMeshLocation002}
   /SetVariables(Particles.MeshOrientation=/Niagara/DynamicInputs/TypeConversions/MakeCustomQuaternionFromBool(
     Bool:NiagaraBool=&SetBoolByIntComparison(A=$Particles.VisibilityTag, B=1, Comparison Type=::NewEnumerator4)
     True Quat:Quat4f=&MultiplyQuaternion(
      Quaternion A=$Particles.SkeletalMeshLocation.SampledRotation
      Quaternion B=&MakeQuatFromAxisAngle(
       Axis=<0 0 1>
       AngleInDegrees=&MaskFloatBySpawnGroup(Masked Float=215, Passthrough Float=35, Spawn Group Mask=7)
      )
     )
     False Quat:Quat4f=$Particles.MeshOrientation
    )) {isSetParameters=true}
   &DynamicMaterialParameters(
    Index 0 Param 1=&Multiply_Float(A=$Particles.Timeline.Time, B=-0.5)
    Index 0 Param 2=&Multiply_Float(A=$Particles.Timeline.Time, B=0.05)
    Index 0 Param 3=&FloatFromCurve(FloatCurve=DataInterface() {propertyValues {Curve {keys [<at 0 0.3>,<at 1 0.1>]}}}, CurveIndex=$System.ReactiveValue)
    Index 1 Param 1=&MakeFloatFromVector(Vector=$Particles.SkeletalMeshLocation001.SampledPosition, Channel=::NewEnumerator0)
    Index 1 Param 2=&MakeFloatFromVector(Vector=$Particles.SkeletalMeshLocation001.SampledPosition, Channel=::NewEnumerator1)
    Index 1 Param 3=&MakeFloatFromVector(Vector=$Particles.SkeletalMeshLocation001.SampledPosition, Channel=::NewEnumerator2)
    Index 2 Param 1=&MakeFloatFromVector(Vector=$Particles.SkeletalMeshLocation002.SampledPosition, Channel=::NewEnumerator0)
    Index 2 Param 2=&MakeFloatFromVector(Vector=$Particles.SkeletalMeshLocation002.SampledPosition, Channel=::NewEnumerator1)
    Index 2 Param 3=&MakeFloatFromVector(Vector=$Particles.SkeletalMeshLocation002.SampledPosition, Channel=::NewEnumerator2)
    Module.Index0Param0Write=true
    Module.Index0Param1Write=true
    Module.Index0Param2Write=true
    Module.Index0Param3Write=true
    Module.Index1Param0Write=true
    Module.Index1Param1Write=true
    Module.Index1Param2Write=true
    Module.Index1Param3Write=true
    Module.Index2Param0Write=true
    Module.Index2Param1Write=true
    Module.Index2Param2Write=true
    Module.Index2Param3Write=true
    Write Parameter Index 1=true
    Write Parameter Index 2=true
   )
   /SetVariables() {isSetParameters=true}
   &self:-----------Reactive----------() {module=-----------Reactive----------}
   &ScaleColor(
    Scale RGB=&VectorFromFloat(Value=&FN_FrontEndFloatModifier(
      Gameplay=&FloatFromCurve(
       FloatCurve=DataInterface() {propertyValues {Curve {keys [
           {interpMode=RCIM_Cubic, leaveTangent=1.398, tangentMode=RCTM_Break, time=0, value=1}
           {arriveTangent=3.5001, interpMode=RCIM_Cubic, tangentMode=RCTM_Break, time=0.9, value=6}
           <at 1 1>
          ]}}}
       CurveIndex=$System.ReactiveValue
      )
      Front End=&FloatFromCurve(
       FloatCurve=DataInterface() {propertyValues {Curve {keys [
           {interpMode=RCIM_Cubic, leaveTangent=1.398, tangentMode=RCTM_Break, time=0, value=1}
           {arriveTangent=3.5001, interpMode=RCIM_Cubic, tangentMode=RCTM_Break, time=0.9, value=7}
           <at 1 1>
          ]}}}
       CurveIndex=$System.ReactiveValue
      )
     ))
    ScaleA=true
    ScaleRGB=true
   )
   &ScaleColor(
    Curve Index=$System.ReactiveValue
    Linear Color Curve=DataInterface() {propertyValues {
      AlphaCurve {keys [<at 0 1>,<at 1 1>]}
      BlueCurve {keys [<at 0 2>,<at 1 6>]}
      GreenCurve {keys [<at 0 1>,<at 1 0>]}
      RedCurve {keys [<at 0 0>,<at 1 0.5>]}
     }}
    Scale Mode=::NewEnumerator2
   ) {module=ScaleColor001}
   &ScaleColor(
    Scale RGB=&VectorFromFloat(Value=&FNTimeOfDayScalar(Night Scale=1, Morning Scale=1, Evening Scale=1))
    ScaleA=true
    ScaleRGB=true
   ) {module=ScaleColor002}
   &KillParticles(Kill Particles=/Niagara/DynamicInputs/Helpers/MaskBool(
     Mask Source:NiagaraInt32=1
     Mask Parameter:NiagaraInt32=$Particles.PartHiddenCheckValue
     Mask:NiagaraInt32=1
     Masked Bool:NiagaraBool=$User.bisHandHidden
     Passthrough Bool:NiagaraBool=false
    ))
  ]
  rendererClasses [&NiagaraMeshRendererProperties]
  renderers [
   {rendererClass=&NiagaraMeshRendererProperties, rendererIndex=0}
   {rendererClass=&NiagaraMeshRendererProperties, rendererIndex=1}
  ]
 }
 SoundCube {
  emitterUpdateScript [
   &EmitterState()
   &SpawnBurst_Instantaneous(Spawn Count=9, Use Spawn Probability=false, UseLoopCountLimit=false)
   &SpawnBurst_Instantaneous(Spawn Count=9, Spawn Group=1, Use Spawn Probability=false, UseLoopCountLimit=false) {module=SpawnBurst_Instantaneous001}
   &SpawnBurst_Instantaneous(Spawn Count=9, Spawn Group=3, Use Spawn Probability=false, UseLoopCountLimit=false) {module=SpawnBurst_Instantaneous002}
   &SpawnBurst_Instantaneous(Spawn Count=9, Spawn Group=4, Use Spawn Probability=false, UseLoopCountLimit=false) {module=SpawnBurst_Instantaneous003}
  ]
  particleSpawnScript [&InitializeParticle(
    Color Mode=::NewEnumerator0
    Lifetime=5
    Mesh Scale Mode=::NewEnumerator3
    Mesh Uniform Scale=0.5
    UsePositionOffset=false
    Write Material Random=true
   )]
  particleUpdateScript [
   &ParticleState()
   &ScaleMeshSize(Scale Factor=&MakeVector(
     X=1
     Y=1
     Z=&Waveform(
      Global Input Position=&RandomRangeFloat(Minimum=0, Maximum=15)
      "Waveform [1]"=::NewEnumerator12
      "[1] Phase Scale"=10
      "[1] Frequency"=1
      "[1] Amplitude Min/Max"=&MakeVector2D(X=0, Y=&Lerp_Float(A=10, B=35, Alpha=$System.ReactiveValue))
     )
    ))
   &SkeletalMeshLocation(
    Bone / Socket Index=&SpawnGroup()
    MaskSpawnGroup=false
    Mesh Sampling Type=::NewEnumerator5
    Sampled Position Offset=<0 0 1.5>
    Skeletal Mesh=$User.Skeletal Mesh
    Socket Sampling Mode=::NewEnumerator5
   )
   /SetVariables(Particles.Position=&AddVectorToPosition(
     Position=$Particles.Position
     Vector=&Multiply_VectorByFloat(
      Vector=&RotateVectorByQuaternion(
       Vector To Rotate=&SelectVectorFromArray(
        Direct Array Index=&ModuloInt(A=&ReturnParticleIDAttributes(), B=9)
        Direct Array Mode=::NewEnumerator1
        Vector Selection Array=DataInterface() {propertyValues {FloatData [<1 1 0>, <1 0 0>, <1 -1 0>, <0 1 0>, <0 0 0>, <0 -1 0>, <-1 1 0>, <-1 0 0>, <-1 -1 0>]}}
       )
       Quaternion=$StackContext.SkeletalMeshLocation.SampledRotation
      )
      Float=0.6091
     )
    )) {isSetParameters=true}
   &ScaleColor(
    Scale Alpha=&Waveform(
     Global Input Position=&RandomRangeFloat(Minimum=0, Maximum=15)
     "Waveform [1]"=::NewEnumerator12
     "[1] Phase Scale"=10
     "[1] Frequency"=1
     "[1] Amplitude Min/Max"=<0 0.5>
    )
    ScaleA=true
    ScaleRGB=true
   )
   &ScaleColor(
    Scale Alpha=&FloatFromCurve(
     FloatCurve=DataInterface() {propertyValues {Curve {keys [
         {interpMode=RCIM_Cubic, leaveTangent=0.431, tangentMode=RCTM_Break, time=0, value=1}
         {arriveTangent=3.4786, interpMode=RCIM_Cubic, leaveTangent=-19.0618, tangentMode=RCTM_Break, time=0.9, value=2}
         <at 1 1>
        ]}}}
     CurveIndex=$System.ReactiveValue
    )
    Scale RGB=&VectorFromFloat(Value=&FloatFromCurve(
      FloatCurve=DataInterface() {propertyValues {Curve {keys [
          {interpMode=RCIM_Cubic, leaveTangent=3.4065, tangentMode=RCTM_Break, time=0, value=1}
          {arriveTangent=0.8198, interpMode=RCIM_Cubic, leaveTangent=-19.0618, tangentMode=RCTM_Break, time=0.9, value=4}
          <at 1 1>
         ]}}}
      CurveIndex=$System.ReactiveValue
     ))
    ScaleA=true
    ScaleRGB=true
   ) {module=ScaleColor001}
  ]
  rendererClasses [&NiagaraSpriteRendererProperties, &NiagaraMeshRendererProperties]
  renderers [
   {rendererClass=&NiagaraSpriteRendererProperties}
   {rendererClass=&NiagaraMeshRendererProperties, rendererIndex=1}
  ]
 }
 Spike {
  emitterUpdateScript [
   &EmitterState(
    Inactive Response=::NewEnumerator1
    Life Cycle Mode=::NewEnumerator1
    Loop Behavior=::NewEnumerator1
    Loop Duration Mode=::NewEnumerator1
    UseLoopDelay=false
   )
   &SpawnBurst_Instantaneous(Spawn Count=8, Use Spawn Probability=false, UseLoopCountLimit=false)
  ]
  particleSpawnScript [
   &self:4Group4Particle(SpawnParticleCount=8) {module=4Group4Particle}
   &InitializeParticle(
    Color=<rgba 0.4885 0 0.9 1>
    Lifetime=5
    Mesh Scale=<0.5 0.5 1>
    Mesh Scale Mode=::NewEnumerator4
    Sprite Size=&SelectVector2DFromArray(Direct Array Index=$Particles.GroupID, Vector 2D Selection Array=DataInterface() {propertyValues {FloatData [<20 60>, <6 6>]}})
    Sprite Size Mode=::NewEnumerator4
    UsePositionOffset=false
    Write Material Random=true
   )
   /SetVariables(Particles.VisibilityTag=&SelectIntFromArray(Direct Array Index=$Particles.GroupID, Int Selection Array=DataInterface() {propertyValues {IntData [0,1,1,1]}})) {isSetParameters=true}
  ]
  particleUpdateScript [
   &ParticleState(Kill Particles When Lifetime Has Elapsed=false, Let Infinitely Lived Particles Die When Emitter Deactivates=true)
   &ScaleSpriteSize(
    Scale Factor=&MakeCustomVector2DFromBool(
     Bool=&SetBoolByIntComparison(A=$Particles.GroupID, B=0, Comparison Type=::NewEnumerator4)
     True Vector2D=&Multiply_Vector2D(
      A=&SelectVector2DFromArray(
       Direct Array Index=$Particles.SeqID
       Vector 2D Selection Array=DataInterface() {propertyValues {FloatData [<0.8 0.4>, <0.8 0.5>, <0.8 0.4>, <0.8 0.5>]}}
      )
      B=&MakeVector2D(
       X=1
       Y=&Waveform("Waveform [1]"=::NewEnumerator12, "[1] Phase Scale"=10, "[1] Frequency"=1, "[1] Amplitude Min/Max"=<0.7 1>)
      )
     )
     False Vector2D=<1 1>
    )
    Scale Sprite Size Mode=::NewEnumerator2
   )
   &SkeletalMeshLocation(
    Bone / Socket Index=&SelectIntFromArray(Direct Array Index=$Particles.SeqID, Int Selection Array=DataInterface() {propertyValues {IntData [0,1,3,4]}})
    MaskSpawnGroup=false
    Mesh Sampling Type=::NewEnumerator5
    Skeletal Mesh=$User.Skeletal Mesh
    Socket Sampling Mode=::NewEnumerator5
   )
   /SetVariables(Particles.MeshOrientation=&MultiplyQuaternion(
     Quaternion A=&MultiplyQuaternion(
      Quaternion A=$StackContext.SkeletalMeshLocation.SampledRotation
      Quaternion B=&MakeQuatFromAxisAngle(
       Axis=<1 0 0>
       AngleInDegrees=&SelectFloatFromArray(Direct Array Index=$Particles.SeqID, Float Selection Array=DataInterface() {propertyValues {FloatData [0,0,0,0]}})
      )
     )
     Quaternion B=&MakeQuatFromAxisAngle(Axis=<0 1 0>, AngleInDegrees=0)
    )) {isSetParameters=true}
   /SetVariables(Particles.SpriteAlignment=&RotateVectorByQuaternion(
     Vector To Rotate=&SelectVectorFromArray(Direct Array Index=$Particles.GroupID, Vector Selection Array=DataInterface() {propertyValues {FloatData [<0 0 1>, <1 0 0>]}})
     Quaternion=$Particles.MeshOrientation
    )) {isSetParameters=true, module=SetVariables_#4}
   /SetVariables(Particles.SpriteFacing=&RotateVectorByQuaternion(Vector To Rotate=<0 0 1>, Quaternion=$Particles.MeshOrientation)) {isSetParameters=true}
   &CameraOffset(Camera Offset Amount=5)
   &self:-----------Reactive----------() {module=-----------Reactive----------}
   &ScaleColor(
    Scale Alpha=&FloatFromCurve(
     FloatCurve=DataInterface() {propertyValues {Curve {keys [
         {interpMode=RCIM_Cubic, leaveTangent=2.3077, tangentMode=RCTM_Break, time=0, value=1}
         {arriveTangent=12.0961, interpMode=RCIM_Cubic, tangentMode=RCTM_Break, time=1, value=5}
        ]}}}
     CurveIndex=$System.ReactiveValue
    )
    ScaleA=true
    ScaleRGB=true
   ) {module=ScaleColor001}
   &ScaleSpriteSize(
    Scale Factor=&MakeVector2D(
     X=1
     Y=&FloatFromCurve(
      FloatCurve=DataInterface() {propertyValues {Curve {keys [
          {interpMode=RCIM_Cubic, leaveTangent=0.6131, tangentMode=RCTM_Break, time=0, value=1}
          {arriveTangent=2.9208, interpMode=RCIM_Cubic, tangentMode=RCTM_Break, time=0.9, value=2}
          {arriveTangent=-15.8075, interpMode=RCIM_Cubic, tangentMode=RCTM_Break, time=1, value=1}
         ]}}}
      CurveIndex=$System.ReactiveValue
     )
    )
    Scale Sprite Size Mode=::NewEnumerator2
   ) {module=ScaleSpriteSize001}
   &ScaleColor(
    Scale RGB=&VectorFromFloat(Value=&FNTimeOfDayScalar(Night Scale=1, Morning Scale=1, Evening Scale=1))
    ScaleA=true
    ScaleRGB=true
   )
  ]
  rendererClasses [&NiagaraSpriteRendererProperties]
  renderers [
   {rendererClass=&NiagaraSpriteRendererProperties}
   {rendererClass=&NiagaraSpriteRendererProperties, rendererIndex=1}
  ]
 }
 SpikeWave {
  emitterUpdateScript [
   &EmitterState(
    Inactive Response=::NewEnumerator1
    Life Cycle Mode=::NewEnumerator1
    Loop Behavior=::NewEnumerator1
    Loop Duration Mode=::NewEnumerator1
    UseLoopDelay=false
   )
   /SetVariables(Emitter.TheSpawnRate:NiagaraFloat=&self:ScratchDynamicInput(
     float:NiagaraFloat=&FloatFromCurve(
      FloatCurve=DataInterface() {propertyValues {Curve {keys [{interpMode=RCIM_Cubic, leaveTangent=29.1592, tangentMode=RCTM_Break, time=0, value=0}, <at 1 10>]}}}
      CurveIndex=$System.ReactiveValue
      Scale Curve=&FN_PickFloatDependOnPlatform(TargetPlatform=DataInterface() {propertyValues {Platforms {qualityLevelMask=-29}}}, TargetPlatformActive=0.5, TargetPlatformNoActive=1)
     )
     float001:NiagaraFloat=1
    )) {isSetParameters=true}
   &SpawnRate(SpawnRate=$Emitter.TheSpawnRate, Use Spawn Probability=false)
   &SpawnRate(SpawnGroup=1, SpawnRate=$Emitter.TheSpawnRate, Use Spawn Probability=false) {module=SpawnRate003}
   &SpawnRate(SpawnGroup=3, SpawnRate=$Emitter.TheSpawnRate, Use Spawn Probability=false) {module=SpawnRate002}
   &SpawnRate(SpawnGroup=4, SpawnRate=$Emitter.TheSpawnRate, Use Spawn Probability=false) {module=SpawnRate001}
  ]
  particleSpawnScript [&InitializeParticle(
    Color=<rgba 0.9298 0.1 1 0.3>
    Lifetime=&FloatFromCurve(
     FloatCurve=DataInterface() {propertyValues {Curve {keys [
         {interpMode=RCIM_Cubic, leaveTangent=-0.0748, tangentMode=RCTM_Break, time=0, value=0.3}
         {arriveTangent=-0.0179, interpMode=RCIM_Cubic, tangentMode=RCTM_Break, time=1, value=0.2}
        ]}}}
     CurveIndex=$System.ReactiveValue
    )
    Sprite Size=<20 60>
    Sprite Size Mode=::NewEnumerator4
    UsePositionOffset=false
    Write Material Random=true
   )]
  particleUpdateScript [
   &ParticleState()
   &ScaleSpriteSize(
    Scale Factor=&MaskVector2dBySpawnGroup(
     Masked Vector2D=<0.8 0.4>
     Passthrough Vector2D=&MaskVector2dBySpawnGroup(Masked Vector2D=<0.8 0.4>, Passthrough Vector2D=<0.8 0.5>, Spawn Group Mask=3)
     Spawn Group Mask=0
    )
    Scale Sprite Size Mode=::NewEnumerator2
   )
   &ScaleSpriteSize(
    Scale Sprite Size Mode=::NewEnumerator0
    Uniform Scale Factor=&RandomRangeFloat(Minimum=0.8, Maximum=1.2)
   ) {module=ScaleSpriteSize002}
   &ScaleSpriteSize(Uniform Curve Sprite Scale=DataInterface() {propertyValues {Curve {keys [{tangentMode=RCTM_User, time=0, value=0.4}, {tangentMode=RCTM_User, time=1, value=1.3}]}}}) {module=ScaleSpriteSize001}
   &SkeletalMeshLocation(
    Bone / Socket Index=&SpawnGroup()
    MaskSpawnGroup=false
    Mesh Sampling Type=::NewEnumerator5
    Skeletal Mesh=$User.Skeletal Mesh
    Socket Sampling Mode=::NewEnumerator5
   )
   /SetVariables(Particles.SpriteAlignment=&RotateVectorByQuaternion(
     Vector To Rotate=<0 0 1>
     Quaternion=&MultiplyQuaternion(
      Quaternion A=&MultiplyQuaternion(
       Quaternion A=$StackContext.SkeletalMeshLocation.SampledRotation
       Quaternion B=&MakeQuatFromAxisAngle(
        Axis=<1 0 0>
        AngleInDegrees=&MaskFloatBySpawnGroup(
         Masked Float=0
         Passthrough Float=&MaskFloatBySpawnGroup(Masked Float=0, Passthrough Float=0, Spawn Group Mask=3)
        )
       )
      )
      Quaternion B=&MakeQuatFromAxisAngle(Axis=<0 1 0>, AngleInDegrees=&MaskFloatBySpawnGroup(Masked Float=0, Passthrough Float=0))
     )
    )) {isSetParameters=true, module=SetVariables_#4}
   &CameraOffset(Camera Offset Amount=5)
   &ScaleColor(
    Scale Alpha=&FloatFromCurve(FloatCurve=DataInterface() {propertyValues {Curve {keys [
         {interpMode=RCIM_Cubic, tangentMode=RCTM_User, time=0, value=0}
         {interpMode=RCIM_Cubic, tangentMode=RCTM_User, time=0.3, value=1}
         {tangentMode=RCTM_User, time=1, value=0}
        ]}}})
    ScaleA=true
    ScaleRGB=true
   )
   &self:-----------Reactive----------() {module=-----------Reactive----------}
   &ScaleColor(
    Scale Alpha=&FloatFromCurve(
     FloatCurve=DataInterface() {propertyValues {Curve {keys [
         {interpMode=RCIM_Cubic, tangentMode=RCTM_Break, time=0, value=1}
         {arriveTangent=14.6964, interpMode=RCIM_Cubic, tangentMode=RCTM_Break, time=1, value=10}
        ]}}}
     CurveIndex=$System.ReactiveValue
    )
    ScaleA=true
    ScaleRGB=true
   ) {module=ScaleColor001}
   &ScaleSpriteSize(
    Scale Factor=&MakeVector2D(
     X=1
     Y=&FloatFromCurve(
      FloatCurve=DataInterface() {propertyValues {Curve {keys [
          {interpMode=RCIM_Cubic, tangentMode=RCTM_Break, time=0, value=1}
          {arriveTangent=3.1273, interpMode=RCIM_Cubic, tangentMode=RCTM_Break, time=0.9, value=2}
          <at 1 1>
         ]}}}
      CurveIndex=$System.ReactiveValue
     )
    )
    Scale Sprite Size Mode=::NewEnumerator2
   ) {module=ScaleSpriteSize003}
   &ScaleColor(
    Scale RGB=&VectorFromFloat(Value=&FNTimeOfDayScalar(Night Scale=1, Morning Scale=1, Evening Scale=1))
    ScaleA=true
    ScaleRGB=true
   ) {module=ScaleColor002}
  ]
  rendererClasses [&NiagaraSpriteRendererProperties]
  renderers [
   {rendererClass=&NiagaraSpriteRendererProperties}
   {rendererClass=&NiagaraSpriteRendererProperties, rendererIndex=1}
  ]
 }
}
summary {
 systemName=NS_TechnoA_2820TOD
 emitters [
  {emitterName=Spike, rendererClasses [&NiagaraSpriteRendererProperties]}
  {emitterName=Glow, rendererClasses [&NiagaraSpriteRendererProperties]}
  {emitterName=ShoulderLine, rendererClasses [&NiagaraMeshRendererProperties]}
  {emitterName=SpikeWave, rendererClasses [&NiagaraSpriteRendererProperties]}
  {emitterName=Ring, rendererClasses [&NiagaraMeshRendererProperties]}
  {emitterName=PostWave, rendererClasses [&NiagaraSpriteRendererProperties, &NiagaraMeshRendererProperties]}
  {emitterName=PostLine, rendererClasses [&NiagaraSpriteRendererProperties, &NiagaraMeshRendererProperties]}
  {emitterName=BurstRing, rendererClasses [&NiagaraMeshRendererProperties]}
  {emitterName=SoundCube, rendererClasses [&NiagaraSpriteRendererProperties, &NiagaraMeshRendererProperties]}
  {emitterName=Glow001, rendererClasses [&NiagaraSpriteRendererProperties]}
  {emitterName=PostLine001, rendererClasses [&NiagaraSpriteRendererProperties, &NiagaraMeshRendererProperties]}
  {emitterName=PostWave001, rendererClasses [&NiagaraSpriteRendererProperties, &NiagaraMeshRendererProperties]}
 ]
 userVariables [
  User.bIsEmote {type=NiagaraBool}
  User.bIsFrontEnd {type=NiagaraBool}
  User.bIsFrontEndPreview {type=NiagaraBool}
  User.bisHandHidden {type=NiagaraBool}
  User.bool {type=NiagaraBool}
  User.Kill {type=NiagaraFloat, defaultValue=0, description=""}
  User.Skeletal Mesh {
   type=NiagaraDataInterfaceSkeletalMesh
   defaultValue {
    struct=NiagaraExt_VariableValue_DataInterface
    value {dataInterface=&self:NiagaraDataInterfaceSkeletalMesh_0}
   }
  }
 ]
}
